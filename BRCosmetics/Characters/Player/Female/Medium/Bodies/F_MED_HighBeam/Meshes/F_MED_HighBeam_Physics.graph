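# Physics asset for the HighBeam medium female player body, covering secondary motion of its dangling costume pieces rather than a full ragdoll.
# Short capsule chains form the bow, ties, belt and two right-arm bands, and box pairs form the boot pieces, all with tightly limited twist.
# Chains hang from master bow and belt bodies, calves or right upper arm, which link to pelvis or spine_05 through free-rotation constraints.
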PhysicsAsset /BRCosmetics/Characters/Player/Female/Medium/Bodies/F_MED_HighBeam/Meshes/F_MED_HighBeam_Physics
bodies {
 calf_l [{shapeType=Capsule, length=5, radius=2, shapeName=calf_l_capsule, rotation=<rot 90 0 0>}]
 calf_r [{shapeType=Capsule, length=5, radius=2, shapeName=calf_r_capsule, rotation=<rot 90 0 0>}]
 dyn_band_bot_01 [{shapeType=Capsule, length=3.3916, radius=1.1084, shapeName=dyn_band_bot_01_capsule, center=<2 0 0>}]
 dyn_band_bot_02 [{shapeType=Capsule, length=3.9607, radius=1.0868, shapeName=dyn_band_bot_02_capsule, center=<2 0 0>, rotation=<rot 90 0 0>}]
 dyn_band_bot_03 [{shapeType=Capsule, length=4.0187, radius=1.0868, shapeName=dyn_band_bot_03_capsule, center=<3 0 0>, rotation=<rot 90 0 0>}]
 dyn_band_top_01 [{shapeType=Capsule, length=3, radius=1.5, shapeName=dyn_band_top_01_capsule, center=<2 0 0>}]
 dyn_band_top_02 [{shapeType=Capsule, length=2.5695, radius=1.5, shapeName=dyn_band_top_02_capsule, center=<2 0 0>, rotation=<rot 90 0 0>}]
 dyn_band_top_03 [{shapeType=Capsule, length=3.6055, radius=1.5, shapeName=dyn_band_top_03_capsule, center=<3 0 0>, rotation=<rot 90 0 0>}]
 dyn_belt_01 [{shapeType=Capsule, length=3.9153, radius=1.8266, shapeName=dyn_belt_01_capsule, center=<2 0 0>}]
 dyn_belt_02 [{shapeType=Capsule, length=4.5771, radius=1.8314, shapeName=dyn_belt_02_capsule, center=<2 0 0>, rotation=<rot -5 0.0002 0>}]
 dyn_belt_03 [{shapeType=Capsule, length=2.2686, radius=1.8314, shapeName=dyn_belt_03_capsule, center=<2 0 0>, rotation=<rot -5 0.0002 0>}]
 dyn_belt_MSTR [{shapeType=Capsule, length=0.505, radius=0.505, shapeName=dyn_belt_MSTR_capsule, rotation=<rot 0 -90 0>}]
 dyn_boot_l_01 [{shapeType=Box, extentX=4.0075, extentY=0.9, extentZ=5, shapeName=dyn_boot_l_01_box, center=<2 0 0>, rotation=<rot 0 0.0004 22.5>}]
 dyn_boot_l_02 [{shapeType=Box, extentX=4.0075, extentY=0.9, extentZ=5, shapeName=dyn_boot_l_02_box, center=<2 0 0>, rotation=<rot 0 0.0004 22.5>}]
 dyn_boot_r_01 [{shapeType=Box, extentX=4.0075, extentY=0.9, extentZ=5, shapeName=dyn_boot_r_01_box, center=<2 0 0>, rotation=<rot 0 0.0003 -25.3126>}]
 dyn_boot_r_02 [{shapeType=Box, extentX=4, extentY=0.9, extentZ=5, shapeName=dyn_boot_r_02_box, center=<2 0 0>, rotation=<rot 0 0.0003 -25.3126>}]
 dyn_bow_MSTR [{shapeType=Capsule, length=0.505, radius=0.505, shapeName=dyn_bow_MSTR_capsule, rotation=<rot 0 -90 0>}]
 dyn_bow_l_01 [{shapeType=Capsule, length=1.5968, radius=0.9372, shapeName=dyn_bow_l_01_capsule, center=<1 0 0>, rotation=<rot 90 0 0>}]
 dyn_bow_l_02 [{shapeType=Capsule, length=1.5968, radius=0.9372, shapeName=dyn_bow_l_02_capsule, center=<1 0 0>, rotation=<rot 90 0 0>}]
 dyn_bow_r_01 [{shapeType=Capsule, length=1.5968, radius=0.9372, shapeName=dyn_bow_r_01_capsule, center=<1 0 0>, rotation=<rot 90 0 0>}]
 dyn_bow_r_02 [{shapeType=Capsule, length=1.5968, radius=0.9372, shapeName=dyn_bow_r_02_capsule, center=<1 0 0>, rotation=<rot 90 0 0>}]
 dyn_tie_l_01 [{shapeType=Capsule, length=1.5, radius=0.5, shapeName=dyn_tie_l_01_capsule, center=<2 0 0>, rotation=<rot 90 0 0>}]
 dyn_tie_l_02 [{shapeType=Capsule, length=1.5, radius=0.5, shapeName=dyn_tie_l_02_capsule, center=<2 0 0>, rotation=<rot 90 0 0>}]
 dyn_tie_r_01 [{shapeType=Capsule, length=1.5, radius=0.5, shapeName=dyn_tie_r_01_capsule, center=<2 0 0>, rotation=<rot 90 0 0>}]
 dyn_tie_r_02 [{shapeType=Capsule, length=1.5, radius=0.5, shapeName=dyn_tie_r_02_capsule, center=<2 0 0>, rotation=<rot 90 0 0>}]
 pelvis [{shapeType=Capsule, length=19.723, radius=9.8173, shapeName=pelvis_capsule, center=<-3.471 -0.5522 0>}]
 spine_05 [
  {shapeType=Sphere, radius=11.5222, shapeName=spine_05_sphere, center=<-8.5228 2.8077 0.3885>}
  {
   shapeType=Capsule
   length=17.0419
   radius=10.5954
   shapeName=spine_05_capsule
   center=<-7.6762 -2.9473 3.507>
   rotation=<rot -0.4814 -87.8842 -89.871>
  }
 ]
 upperarm_twist_01_r [{shapeType=Capsule, length=8.5278, radius=5.4793, shapeName=upperarm_twist_01_r_capsule, center=<0 0.4978 0.6427>, rotation=<rot 90 0 0>}]
 upperarm_twist_02_r [{
   shapeType=Capsule
   length=11.2185
   radius=5.1507
   shapeName=upperarm_twist_02_r_capsule
   center=<5.8516 0.3315 0.5119>
   rotation=<rot -0.0001 -89.9999 -95>
  }]
}
constraints [
 {bone1Name=dyn_bow_l_01, bone2Name=dyn_bow_MSTR, swing1LimitDegrees=1, swing2LimitDegrees=10, twistLimitDegrees=0.5}
 {bone1Name=dyn_bow_l_02, bone2Name=dyn_bow_l_01, swing1LimitDegrees=1, swing2LimitDegrees=10, twistLimitDegrees=0.5}
 {bone1Name=dyn_bow_r_01, bone2Name=dyn_bow_MSTR, swing1LimitDegrees=2, swing2LimitDegrees=15, twistLimitDegrees=0.5}
 {bone1Name=dyn_bow_r_02, bone2Name=dyn_bow_r_01, swing1LimitDegrees=2, swing2LimitDegrees=15, twistLimitDegrees=0.5}
 {bone1Name=dyn_tie_l_01, bone2Name=dyn_bow_MSTR, swing1LimitDegrees=0.3, swing2LimitDegrees=10, twistLimitDegrees=0.1}
 {bone1Name=dyn_tie_l_02, bone2Name=dyn_tie_l_01, swing1LimitDegrees=0.3, swing2LimitDegrees=10, twistLimitDegrees=0.1}
 {bone1Name=dyn_tie_r_01, bone2Name=dyn_bow_MSTR, swing1LimitDegrees=0.3, swing2LimitDegrees=10, twistLimitDegrees=0.1}
 {bone1Name=dyn_tie_r_02, bone2Name=dyn_tie_r_01, swing1LimitDegrees=0.3, swing2LimitDegrees=10, twistLimitDegrees=0.1}
 {bone1Name=dyn_band_bot_01, bone2Name=upperarm_twist_01_r, swing1LimitDegrees=60, swing2LimitDegrees=60, twistLimitDegrees=2}
 {bone1Name=dyn_band_bot_02, bone2Name=dyn_band_bot_01, swing1LimitDegrees=60, swing2LimitDegrees=60, twistLimitDegrees=2}
 {bone1Name=dyn_band_bot_03, bone2Name=dyn_band_bot_02, swing1LimitDegrees=60, swing2LimitDegrees=60, twistLimitDegrees=2}
 {bone1Name=dyn_band_top_01, bone2Name=upperarm_twist_01_r, swing1LimitDegrees=60, swing2LimitDegrees=60, twistLimitDegrees=2}
 {bone1Name=dyn_band_top_02, bone2Name=dyn_band_top_01, swing1LimitDegrees=60, swing2LimitDegrees=60, twistLimitDegrees=2}
 {bone1Name=dyn_band_top_03, bone2Name=dyn_band_top_02, swing1LimitDegrees=60, swing2LimitDegrees=60, twistLimitDegrees=2}
 {bone1Name=dyn_boot_l_02, bone2Name=dyn_boot_l_01, swing1LimitDegrees=15, swing2LimitDegrees=8, twistLimitDegrees=0.2}
 {bone1Name=dyn_boot_r_02, bone2Name=dyn_boot_r_01, swing1LimitDegrees=15, swing2LimitDegrees=8, twistLimitDegrees=0.2}
 {bone1Name=dyn_boot_r_01, bone2Name=calf_r, swing1LimitDegrees=15, swing2LimitDegrees=8, twistLimitDegrees=0.2}
 {bone1Name=dyn_boot_l_01, bone2Name=calf_l, swing1LimitDegrees=15, swing2LimitDegrees=8, twistLimitDegrees=0.2}
 {bone1Name=dyn_belt_03, bone2Name=dyn_belt_02, swing1LimitDegrees=10, swing2LimitDegrees=20, twistLimitDegrees=0.2}
 {bone1Name=dyn_belt_02, bone2Name=dyn_belt_01, swing1LimitDegrees=10, swing2LimitDegrees=20, twistLimitDegrees=0.2}
 {bone1Name=dyn_belt_01, bone2Name=dyn_belt_MSTR, swing1LimitDegrees=10, swing2LimitDegrees=20, twistLimitDegrees=0.2}
 {
  bone1Name=upperarm_twist_01_r
  bone2Name=spine_05
  swing1LimitDegrees=45
  swing1Motion=Free
  swing2LimitDegrees=45
  swing2Motion=Free
  twistLimitDegrees=45
  twistMotion=Free
 }
 {
  bone1Name=upperarm_twist_02_r
  bone2Name=spine_05
  swing1LimitDegrees=45
  swing1Motion=Free
  swing2LimitDegrees=45
  swing2Motion=Free
  twistLimitDegrees=45
  twistMotion=Free
 }
 {
  bone1Name=spine_05
  bone2Name=pelvis
  swing1LimitDegrees=45
  swing1Motion=Free
  swing2LimitDegrees=45
  swing2Motion=Free
  twistLimitDegrees=45
  twistMotion=Free
 }
 {
  bone1Name=dyn_bow_MSTR
  bone2Name=pelvis
  swing1LimitDegrees=45
  swing1Motion=Free
  swing2LimitDegrees=45
  swing2Motion=Free
  twistLimitDegrees=45
  twistMotion=Free
 }
 {
  bone1Name=calf_l
  bone2Name=pelvis
  swing1LimitDegrees=45
  swing1Motion=Free
  swing2LimitDegrees=45
  swing2Motion=Free
  twistLimitDegrees=45
  twistMotion=Free
 }
 {
  bone1Name=calf_r
  bone2Name=pelvis
  swing1LimitDegrees=45
  swing1Motion=Free
  swing2LimitDegrees=45
  swing2Motion=Free
  twistLimitDegrees=45
  twistMotion=Free
 }
 {
  bone1Name=dyn_belt_MSTR
  bone2Name=pelvis
  swing1LimitDegrees=45
  swing1Motion=Free
  swing2LimitDegrees=45
  swing2Motion=Free
  twistLimitDegrees=45
  twistMotion=Free
 }
]
