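# Animation Blueprint for the HighBeam female medium body cosmetic, adding secondary motion to its bands, belts, bags, strings, tassels, boots and dagger.
# It copies the parent mesh pose through a deformer Control Rig, then chains per-bone AnimDynamics, some with spherical limits against body bones.
# A second Control Rig takes Jumps, DBNO, ADS and Crouch values.
# Bone offsets and RigidBody simulations follow, using a gravity override and skydive-dependent blending.
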AnimBlueprint /BRCosmetics/Characters/Player/Female/Medium/Bodies/F_MED_HighBeam/Meshes/F_MED_HighBeam_AnimBP
target_skeleton=/Game/Characters/Player/Female/Medium/Bodies/F_MED_HighBeam/Meshes/F_MED_HighBeam_Skeleton
graphs [
 AnimGraph {nodes [
   n1 = Anim Dynamics - Bone: dyn_band_top_01(Alpha=1, ExternalForce="0,0,0") {
    angularDampingOverride=1
    angularSpringConstant=700
    bAngularSpring=true
    bOverrideAngularDamping=true
    bUseGravityOverride=true
    simulationSpace=RootRelative
    alphaBoolBlend {blendOption=Linear}
    boundBone {boneName=dyn_band_top_01}
    componentAppliedLinearAccClamp=<10000 10000 10000>
    componentLinearAccScale=<0.1 0.1 0.3>
    componentLinearVelScale=<0.1 0.1 0.3>
    gravityOverride {z=-980}
    physicsBodyDefinitions [{
      boundBone {boneName=dyn_band_top_01}
      boxExtents=<15 15 40>
      constraintSetup {angularLimitsMax=<0 30 30>, angularLimitsMin=<0 -30 -30>, angularTarget=<0 0 0>}
      localJointOffset=<0.5 0 0>
     }]
   }
   n2 = Anim Dynamics - Bone: dyn_band_top_02(ComponentPose=@n1, Alpha=1, ExternalForce="0,0,0") {
    angularDampingOverride=1
    angularSpringConstant=700
    bAngularSpring=true
    bOverrideAngularDamping=true
    bUseGravityOverride=true
    simulationSpace=RootRelative
    alphaBoolBlend {blendOption=Linear}
    boundBone {boneName=dyn_band_top_02}
    componentAppliedLinearAccClamp=<10000 10000 10000>
    componentLinearAccScale=<0.1 0.1 0.3>
    componentLinearVelScale=<0.1 0.1 0.3>
    gravityOverride {z=-980}
    physicsBodyDefinitions [{
      boundBone {boneName=dyn_band_top_02}
      boxExtents=<5 5 20>
      constraintSetup {angularLimitsMax=<10 15 15>, angularLimitsMin=<-10 -15 -15>, angularTarget=<0 0 0>}
      localJointOffset=<0.5 0 0>
     }]
   }
   n3 = Anim Dynamics - Bone: dyn_band_top_03(ComponentPose=@n2, Alpha=1, ExternalForce="0,0,0") {
    angularDampingOverride=1
    angularSpringConstant=700
    bAngularSpring=true
    bOverrideAngularDamping=true
    bUseGravityOverride=true
    simulationSpace=RootRelative
    alphaBoolBlend {blendOption=Linear}
    boundBone {boneName=dyn_band_top_03}
    componentAppliedLinearAccClamp=<10000 10000 10000>
    componentLinearAccScale=<0.1 0.1 0.3>
    componentLinearVelScale=<0.1 0.1 0.3>
    gravityOverride {z=-980}
    physicsBodyDefinitions [{
      boundBone {boneName=dyn_band_top_03}
      boxExtents=<5 5 20>
      constraintSetup {angularLimitsMax=<10 15 15>, angularLimitsMin=<-10 -15 -15>, angularTarget=<0 0 0>}
      localJointOffset=<0.5 0 0>
     }]
   }
   n4 = Anim Dynamics - Bone: dyn_band_bot_01(ComponentPose=@n3, Alpha=1, ExternalForce="0,0,0") {
    angularDampingOverride=1
    angularSpringConstant=700
    bAngularSpring=true
    bOverrideAngularDamping=true
    bUseGravityOverride=true
    simulationSpace=RootRelative
    alphaBoolBlend {blendOption=Linear}
    boundBone {boneName=dyn_band_bot_01}
    componentAppliedLinearAccClamp=<10000 10000 10000>
    componentLinearAccScale=<0.1 0.1 0.3>
    componentLinearVelScale=<0.1 0.1 0.3>
    gravityOverride {z=-980}
    physicsBodyDefinitions [{
      boundBone {boneName=dyn_band_bot_01}
      boxExtents=<15 15 40>
      constraintSetup {angularLimitsMax=<0 30 30>, angularLimitsMin=<0 -30 -30>, angularTarget=<0 0 0>}
      localJointOffset=<0.5 0 0>
     }]
   }
   n5 = Anim Dynamics - Bone: dyn_band_bot_02(ComponentPose=@n4, Alpha=1, ExternalForce="0,0,0") {
    angularDampingOverride=1
    angularSpringConstant=700
    bAngularSpring=true
    bOverrideAngularDamping=true
    bUseGravityOverride=true
    simulationSpace=RootRelative
    alphaBoolBlend {blendOption=Linear}
    boundBone {boneName=dyn_band_bot_02}
    componentAppliedLinearAccClamp=<10000 10000 10000>
    componentLinearAccScale=<0.1 0.1 0.3>
    componentLinearVelScale=<0.1 0.1 0.3>
    gravityOverride {z=-980}
    physicsBodyDefinitions [{
      boundBone {boneName=dyn_band_bot_02}
      boxExtents=<5 5 20>
      constraintSetup {angularLimitsMax=<10 15 15>, angularLimitsMin=<-10 -15 -15>, angularTarget=<0 0 0>}
      localJointOffset=<0.5 0 0>
     }]
   }
   Anim Dynamics - Bone: dyn_band_bot_03(ComponentPose=@n5, Alpha=1, ExternalForce="0,0,0") {
    angularDampingOverride=1
    angularSpringConstant=700
    bAngularSpring=true
    bOverrideAngularDamping=true
    bUseGravityOverride=true
    simulationSpace=RootRelative
    alphaBoolBlend {blendOption=Linear}
    boundBone {boneName=dyn_band_bot_03}
    componentAppliedLinearAccClamp=<10000 10000 10000>
    componentLinearAccScale=<0.1 0.1 0.3>
    componentLinearVelScale=<0.1 0.1 0.3>
    gravityOverride {z=-980}
    physicsBodyDefinitions [{
      boundBone {boneName=dyn_band_bot_03}
      boxExtents=<5 5 20>
      constraintSetup {angularLimitsMax=<10 15 15>, angularLimitsMin=<-10 -15 -15>, angularTarget=<0 0 0>}
      localJointOffset=<0.5 0 0>
     }]
   }
   RigidBody!(ComponentPose, Alpha, OverrideWorldGravity:Vector="0,0,0", ExternalForce) -> (Pose) {
    bClampLinearTranslationLimitToRefPose=true
    bOverrideWorldGravity=true
    overridePhysicsAsset=&/F_MED_HighBeam_Physics
    simulationSpace=BaseBoneSpace
    alphaBoolBlend {blendOption=Linear}
    baseBoneRef {boneName=spine_02}
    componentAppliedLinearAccClamp=<3000 3000 3000>
    componentLinearAccScale=<0.5 0.3 0.3>
    componentLinearVelScale=<0.5 0.3 0.3>
    simSpaceSettings {maxAngularVelocity=3, maxLinearAcceleration=700, maxLinearVelocity=500, velocityScaleZ=0, worldAlpha=1, externalLinearDragV=<2 0.2>}
   }
   n6 = Control Rig(Source=Copy Pose From Mesh() {bUseAttachedParent=true}) {
    alphaBoolBlend {blendOption=Linear}
    controlRigAssetReference {blueprintRigClass=&/F_MED_HighBeam_CtrlRig_Deformer.F_MED_HighBeam_CtrlRig_Deformer_C}
   }
   n7 = Anim Dynamics - Bone: dyn_belt_end_l(ComponentPose=Local To Component(LocalPose=@n6), Alpha=1, ExternalForce="0,0,0") {
    angularDampingOverride=0.9999
    bGravityOverrideInSimSpace=true
    bOverrideAngularDamping=true
    bOverrideLinearDamping=true
    gravityScale=0.25
    linearDampingOverride=0.9999
    simulationSpace=RootRelative
    alphaBoolBlend {blendOption=Linear}
    boundBone {boneName=dyn_belt_end_l}
    componentAppliedLinearAccClamp=<900 900 900>
    componentLinearAccScale=<0.25 0.25 0.25>
    componentLinearVelScale=<0.25 0.25 0.25>
    physicsBodyDefinitions [{
      sphereCollisionRadius=0
      boundBone {boneName=dyn_belt_end_l}
      boxExtents=<10 3 5>
      constraintSetup {angularLimitsMax=<10 0 15>, angularLimitsMin=<-10 0 -15>, angularTarget=<0 0 0>}
      localJointOffset=<3 0 0>
     }]
    relativeSpaceBone {boneName=root}
   }
   n8 = Anim Dynamics - Bone: dyn_belt_endfore_l(ComponentPose=@n7, Alpha=1, ExternalForce="0,0,0") {
    angularDampingOverride=0.9999
    bGravityOverrideInSimSpace=true
    bOverrideAngularDamping=true
    bOverrideLinearDamping=true
    gravityScale=0.25
    linearDampingOverride=0.9999
    simulationSpace=RootRelative
    alphaBoolBlend {blendOption=Linear}
    boundBone {boneName=dyn_belt_endfore_l}
    componentAppliedLinearAccClamp=<900 900 900>
    componentLinearAccScale=<0.25 0.25 0.25>
    componentLinearVelScale=<0.25 0.25 0.25>
    physicsBodyDefinitions [{
      sphereCollisionRadius=0
      boundBone {boneName=dyn_belt_endfore_l}
      boxExtents=<10 3 5>
      constraintSetup {angularLimitsMax=<10 0 15>, angularLimitsMin=<-10 0 -15>, angularTarget=<0 0 0>}
      localJointOffset=<3 0 0>
     }]
    relativeSpaceBone {boneName=root}
   }
   n9 = Anim Dynamics - Bone: dyn_belt_end_r(ComponentPose=@n8, Alpha=1, ExternalForce="0,0,0") {
    angularDampingOverride=0.9999
    bGravityOverrideInSimSpace=true
    bOverrideAngularDamping=true
    bOverrideLinearDamping=true
    gravityScale=0.25
    linearDampingOverride=0.9999
    simulationSpace=RootRelative
    alphaBoolBlend {blendOption=Linear}
    boundBone {boneName=dyn_belt_end_r}
    componentAppliedLinearAccClamp=<900 900 900>
    componentLinearAccScale=<0.25 0.25 0.25>
    componentLinearVelScale=<0.25 0.25 0.25>
    physicsBodyDefinitions [{
      sphereCollisionRadius=0
      boundBone {boneName=dyn_belt_end_r}
      boxExtents=<10 3 5>
      constraintSetup {angularLimitsMax=<10 0 15>, angularLimitsMin=<-10 0 -15>, angularTarget=<0 0 0>}
      localJointOffset=<3 0 0>
     }]
    relativeSpaceBone {boneName=root}
   }
   n10 = Anim Dynamics - Bone: dyn_belt_endfore_r(ComponentPose=@n9, Alpha=1, ExternalForce="0,0,0") {
    angularDampingOverride=0.9999
    bGravityOverrideInSimSpace=true
    bOverrideAngularDamping=true
    bOverrideLinearDamping=true
    gravityScale=0.25
    linearDampingOverride=0.9999
    simulationSpace=RootRelative
    alphaBoolBlend {blendOption=Linear}
    boundBone {boneName=dyn_belt_endfore_r}
    componentAppliedLinearAccClamp=<900 900 900>
    componentLinearAccScale=<0.25 0.25 0.25>
    componentLinearVelScale=<0.25 0.25 0.25>
    physicsBodyDefinitions [{
      sphereCollisionRadius=0
      boundBone {boneName=dyn_belt_endfore_r}
      boxExtents=<10 3 5>
      constraintSetup {angularLimitsMax=<10 0 15>, angularLimitsMin=<-10 0 -15>, angularTarget=<0 0 0>}
      localJointOffset=<3 0 0>
     }]
    relativeSpaceBone {boneName=root}
   }
   n11 = Anim Dynamics - Bone: dyn_belt_end_c(ComponentPose=@n10, Alpha=1, ExternalForce="0,0,0") {
    angularDampingOverride=0.9999
    bGravityOverrideInSimSpace=true
    bOverrideAngularDamping=true
    bOverrideLinearDamping=true
    gravityScale=0.25
    linearDampingOverride=0.9999
    simulationSpace=RootRelative
    alphaBoolBlend {blendOption=Linear}
    boundBone {boneName=dyn_belt_end_c}
    componentAppliedLinearAccClamp=<900 900 900>
    componentLinearAccScale=<0.25 0.25 0.25>
    componentLinearVelScale=<0.25 0.25 0.25>
    physicsBodyDefinitions [{
      sphereCollisionRadius=0
      boundBone {boneName=dyn_belt_end_c}
      boxExtents=<10 3 5>
      constraintSetup {angularLimitsMax=<10 0 15>, angularLimitsMin=<-10 0 -15>, angularTarget=<0 0 0>}
      localJointOffset=<3 0 0>
     }]
    relativeSpaceBone {boneName=root}
   }
   n12 = Anim Dynamics - Bone: dyn_bootbelt_l_01(ComponentPose=@n11, Alpha=1, ExternalForce="0,0,0") {
    angularDampingOverride=0.9999
    bGravityOverrideInSimSpace=true
    bOverrideAngularDamping=true
    bOverrideLinearDamping=true
    gravityScale=0.25
    linearDampingOverride=0.9999
    simulationSpace=RootRelative
    alphaBoolBlend {blendOption=Linear}
    boundBone {boneName=dyn_bootbelt_l_01}
    componentAppliedLinearAccClamp=<900 900 900>
    componentLinearAccScale=<0.25 0.25 0.25>
    componentLinearVelScale=<0.25 0.25 0.25>
    physicsBodyDefinitions [{
      sphereCollisionRadius=0
      boundBone {boneName=dyn_bootbelt_l_01}
      boxExtents=<10 3 5>
      constraintSetup {angularLimitsMax=<10 0 15>, angularLimitsMin=<-10 0 -15>, angularTarget=<0 0 0>}
      localJointOffset=<3 0 0>
     }]
    relativeSpaceBone {boneName=root}
   }
   n13 = Anim Dynamics - Bone: dyn_bootbelt_l_02(ComponentPose=@n12, Alpha=1, ExternalForce="0,0,0") {
    angularDampingOverride=0.9999
    bGravityOverrideInSimSpace=true
    bOverrideAngularDamping=true
    bOverrideLinearDamping=true
    gravityScale=0.25
    linearDampingOverride=0.9999
    simulationSpace=RootRelative
    alphaBoolBlend {blendOption=Linear}
    boundBone {boneName=dyn_bootbelt_l_02}
    componentAppliedLinearAccClamp=<900 900 900>
    componentLinearAccScale=<0.25 0.25 0.25>
    componentLinearVelScale=<0.25 0.25 0.25>
    physicsBodyDefinitions [{
      sphereCollisionRadius=0
      boundBone {boneName=dyn_bootbelt_l_02}
      boxExtents=<10 3 5>
      constraintSetup {angularLimitsMax=<10 0 15>, angularLimitsMin=<-10 0 -15>, angularTarget=<0 0 0>}
      localJointOffset=<3 0 0>
     }]
    relativeSpaceBone {boneName=root}
   }
   n14 = Anim Dynamics - Bone: dyn_bootbelt_r_01(ComponentPose=@n13, Alpha=1, ExternalForce="0,0,0") {
    angularDampingOverride=0.9999
    bGravityOverrideInSimSpace=true
    bOverrideAngularDamping=true
    bOverrideLinearDamping=true
    gravityScale=0.25
    linearDampingOverride=0.9999
    simulationSpace=RootRelative
    alphaBoolBlend {blendOption=Linear}
    boundBone {boneName=dyn_bootbelt_r_01}
    componentAppliedLinearAccClamp=<900 900 900>
    componentLinearAccScale=<0.25 0.25 0.25>
    componentLinearVelScale=<0.25 0.25 0.25>
    physicsBodyDefinitions [{
      sphereCollisionRadius=0
      boundBone {boneName=dyn_bootbelt_r_01}
      boxExtents=<10 3 5>
      constraintSetup {angularLimitsMax=<10 0 15>, angularLimitsMin=<-10 0 -15>, angularTarget=<0 0 0>}
      localJointOffset=<3 0 0>
     }]
    relativeSpaceBone {boneName=root}
   }
   n15 = Anim Dynamics - Bone: dyn_bootbelt_r_02(ComponentPose=@n14, Alpha=1, ExternalForce="0,0,0") {
    angularDampingOverride=0.9999
    bGravityOverrideInSimSpace=true
    bOverrideAngularDamping=true
    bOverrideLinearDamping=true
    gravityScale=0.25
    linearDampingOverride=0.9999
    simulationSpace=RootRelative
    alphaBoolBlend {blendOption=Linear}
    boundBone {boneName=dyn_bootbelt_r_02}
    componentAppliedLinearAccClamp=<900 900 900>
    componentLinearAccScale=<0.25 0.25 0.25>
    componentLinearVelScale=<0.25 0.25 0.25>
    physicsBodyDefinitions [{
      sphereCollisionRadius=0
      boundBone {boneName=dyn_bootbelt_r_02}
      boxExtents=<10 3 5>
      constraintSetup {angularLimitsMax=<10 0 15>, angularLimitsMin=<-10 0 -15>, angularTarget=<0 0 0>}
      localJointOffset=<3 0 0>
     }]
    relativeSpaceBone {boneName=root}
   }
   n16 = Anim Dynamics - Bone: dyn_dagger_02(ComponentPose=@n15, Alpha=0.2, ExternalForce="0,0,0") {
    angularDampingOverride=1
    angularSpringConstant=20000
    bAngularSpring=true
    bOverrideAngularDamping=true
    bUseGravityOverride=true
    gravityScale=0.35
    simulationSpace=RootRelative
    alphaBoolBlend {blendOption=Linear}
    boundBone {boneName=dyn_dagger_02}
    componentLinearAccScale=<0.7 0.7 0.7>
    componentLinearVelScale=<0.7 0.7 0.7>
    gravityOverride {z=-980}
    physicsBodyDefinitions [{
      sphereCollisionRadius=0
      boundBone {boneName=dyn_dagger_02}
      boxExtents=<20 30 30>
      constraintSetup {angularLimitsMax=<0 5 10>, angularLimitsMin=<0 -2 -10>, angularTarget=<1 0 0>}
      localJointOffset=<4 0 0>
     }]
   }
   n17 = Anim Dynamics - Bone: dyn_bag_l_fr_01!(ComponentPose=@n16, Alpha=0.6, GravityOverride:Vector="0,0,0", ExternalForce="0,0,0") -> (Pose) {
    angularDampingOverride=0.9999
    angularSpringConstant=500
    bAngularSpring=true
    bGravityOverrideInSimSpace=true
    bOverrideAngularDamping=true
    bUseGravityOverride=true
    bUseSphericalLimits=true
    simulationSpace=BoneRelative
    alphaBoolBlend {blendOption=Linear}
    boundBone {boneName=dyn_bag_l_fr_01}
    componentAppliedLinearAccClamp=<3500 3500 3500>
    componentLinearAccScale=<1.2 1.2 1.2>
    componentLinearVelScale=<1.2 1.2 1.2>
    physicsBodyDefinitions [{
      sphereCollisionRadius=0
      boundBone {boneName=dyn_bag_l_fr_01}
      boxExtents=<15 30 30>
      constraintSetup {angularLimitsMax=<1 5 15>, angularLimitsMin=<-1 -35 -5>, angularTarget=<0 0 0>}
      localJointOffset=<4 0 0>
     }]
    relativeSpaceBone {boneName=root}
    sphericalLimits [{limitRadius=7, drivingBone {boneName=thigh_l}, sphereLocalOffset=<-5 0 0>}]
   }
   n18 = Anim Dynamics - Bone: dyn_string_fr_01!(ComponentPose=@n17, Alpha=0.8, GravityOverride:Vector="0,0,0", ExternalForce="0,0,0") -> (Pose) {
    angularDampingOverride=0.9999
    angularSpringConstant=500
    bAngularSpring=true
    bGravityOverrideInSimSpace=true
    bOverrideAngularDamping=true
    bUseGravityOverride=true
    simulationSpace=BoneRelative
    alphaBoolBlend {blendOption=Linear}
    boundBone {boneName=dyn_string_fr_01}
    componentAppliedLinearAccClamp=<5000 5000 5000>
    componentLinearAccScale=<1.2 1.2 1.2>
    componentLinearVelScale=<1.2 1.2 1.2>
    physicsBodyDefinitions [{
      sphereCollisionRadius=0
      boundBone {boneName=dyn_string_fr_01}
      boxExtents=<15 20 20>
      constraintSetup {angularLimitsMax=<1 5 40>, angularLimitsMin=<-1 -90 -45>, angularTarget=<0 0 0>}
      localJointOffset=<5 0 0>
     }]
    relativeSpaceBone {boneName=root}
   }
   n19 = Anim Dynamics - Bone: dyn_string_fr_02!(ComponentPose=@n18, Alpha=0.5, GravityOverride:Vector="0,0,0", ExternalForce="0,0,0") -> (Pose) {
    angularDampingOverride=0.9999
    angularSpringConstant=500
    bAngularSpring=true
    bGravityOverrideInSimSpace=true
    bOverrideAngularDamping=true
    bUseGravityOverride=true
    simulationSpace=BoneRelative
    alphaBoolBlend {blendOption=Linear}
    boundBone {boneName=dyn_string_fr_02}
    componentAppliedLinearAccClamp=<5000 5000 5000>
    componentLinearAccScale=<1.2 1.2 1.2>
    componentLinearVelScale=<1.2 1.2 1.2>
    physicsBodyDefinitions [{
      sphereCollisionRadius=0
      boundBone {boneName=dyn_string_fr_02}
      boxExtents=<15 20 20>
      constraintSetup {angularLimitsMax=<1 8 40>, angularLimitsMin=<-1 -90 -45>, angularTarget=<0 0 0>}
      localJointOffset=<5 0 0>
     }]
    relativeSpaceBone {boneName=root}
   }
   n20 = Anim Dynamics - Bone: dyn_string_fr_03!(ComponentPose=@n19, Alpha=0.5, GravityOverride:Vector="0,0,0", ExternalForce="0,0,0") -> (Pose) {
    angularDampingOverride=0.9999
    angularSpringConstant=500
    bAngularSpring=true
    bGravityOverrideInSimSpace=true
    bOverrideAngularDamping=true
    bUseGravityOverride=true
    simulationSpace=BoneRelative
    alphaBoolBlend {blendOption=Linear}
    boundBone {boneName=dyn_string_fr_03}
    componentAppliedLinearAccClamp=<5000 5000 5000>
    componentLinearAccScale=<1.2 1.2 1.2>
    componentLinearVelScale=<1.2 1.2 1.2>
    physicsBodyDefinitions [{
      sphereCollisionRadius=0
      boundBone {boneName=dyn_string_fr_03}
      boxExtents=<15 20 20>
      constraintSetup {angularLimitsMax=<1 8 40>, angularLimitsMin=<-1 -90 -45>, angularTarget=<0 0 0>}
      localJointOffset=<5 0 0>
     }]
    relativeSpaceBone {boneName=root}
   }
   n21 = Anim Dynamics - Bone: dyn_string_bk_01!(ComponentPose=@n20, Alpha=0.8, GravityOverride:Vector="0,0,0", ExternalForce="0,0,0") -> (Pose) {
    angularDampingOverride=0.9999
    angularSpringConstant=500
    bAngularSpring=true
    bGravityOverrideInSimSpace=true
    bOverrideAngularDamping=true
    bUseGravityOverride=true
    simulationSpace=BoneRelative
    alphaBoolBlend {blendOption=Linear}
    boundBone {boneName=dyn_string_bk_01}
    componentAppliedLinearAccClamp=<5000 5000 5000>
    componentLinearAccScale=<1.2 1.2 1.2>
    componentLinearVelScale=<1.2 1.2 1.2>
    physicsBodyDefinitions [{
      sphereCollisionRadius=0
      boundBone {boneName=dyn_string_bk_01}
      boxExtents=<15 20 20>
      constraintSetup {angularLimitsMax=<1 5 40>, angularLimitsMin=<-1 -90 -45>, angularTarget=<0 0 0>}
      localJointOffset=<5 0 0>
     }]
    relativeSpaceBone {boneName=root}
   }
   n22 = Anim Dynamics - Bone: dyn_string_bk_02!(ComponentPose=@n21, Alpha=0.5, GravityOverride:Vector="0,0,0", ExternalForce="0,0,0") -> (Pose) {
    angularDampingOverride=0.9999
    angularSpringConstant=500
    bAngularSpring=true
    bGravityOverrideInSimSpace=true
    bOverrideAngularDamping=true
    bUseGravityOverride=true
    simulationSpace=BoneRelative
    alphaBoolBlend {blendOption=Linear}
    boundBone {boneName=dyn_string_bk_02}
    componentAppliedLinearAccClamp=<5000 5000 5000>
    componentLinearAccScale=<1.2 1.2 1.2>
    componentLinearVelScale=<1.2 1.2 1.2>
    physicsBodyDefinitions [{
      sphereCollisionRadius=0
      boundBone {boneName=dyn_string_bk_02}
      boxExtents=<15 20 20>
      constraintSetup {angularLimitsMax=<1 8 40>, angularLimitsMin=<-1 -90 -45>, angularTarget=<0 0 0>}
      localJointOffset=<5 0 0>
     }]
    relativeSpaceBone {boneName=root}
   }
   n23 = Anim Dynamics - Bone: dyn_string_bk_03!(ComponentPose=@n22, Alpha=0.5, GravityOverride:Vector="0,0,0", ExternalForce="0,0,0") -> (Pose) {
    angularDampingOverride=0.9999
    angularSpringConstant=500
    bAngularSpring=true
    bGravityOverrideInSimSpace=true
    bOverrideAngularDamping=true
    bUseGravityOverride=true
    simulationSpace=BoneRelative
    alphaBoolBlend {blendOption=Linear}
    boundBone {boneName=dyn_string_bk_03}
    componentAppliedLinearAccClamp=<5000 5000 5000>
    componentLinearAccScale=<1.2 1.2 1.2>
    componentLinearVelScale=<1.2 1.2 1.2>
    physicsBodyDefinitions [{
      sphereCollisionRadius=0
      boundBone {boneName=dyn_string_bk_03}
      boxExtents=<15 20 20>
      constraintSetup {angularLimitsMax=<1 8 40>, angularLimitsMin=<-1 -90 -45>, angularTarget=<0 0 0>}
      localJointOffset=<5 0 0>
     }]
    relativeSpaceBone {boneName=root}
   }
   n24 = Anim Dynamics - Bone: dyn_tassel_fr!(ComponentPose=@n23, Alpha=1, GravityOverride:Vector="0,0,0", ExternalForce="0,0,0") -> (Pose) {
    angularDampingOverride=0.9999
    angularSpringConstant=500
    bAngularSpring=true
    bGravityOverrideInSimSpace=true
    bOverrideAngularDamping=true
    bUseGravityOverride=true
    bUsePlanarLimit=false
    bUseSphericalLimits=true
    simulationSpace=BoneRelative
    alphaBoolBlend {blendOption=Linear}
    boundBone {boneName=dyn_tassel_fr}
    componentAppliedLinearAccClamp=<3500 3500 3500>
    componentLinearAccScale=<1.2 1.2 1.2>
    componentLinearVelScale=<1.2 1.2 1.2>
    physicsBodyDefinitions [{
      collisionType=CustomSphere
      sphereCollisionRadius=2
      boundBone {boneName=dyn_tassel_fr}
      boxExtents=<15 30 30>
      constraintSetup {angularLimitsMax=<1 5 15>, angularLimitsMin=<-1 -35 -15>, angularTarget=<0 0 0>}
      localJointOffset=<4 0 0>
     }]
    relativeSpaceBone {boneName=root}
    sphericalLimits [{limitRadius=10, drivingBone {boneName=thigh_l}, sphereLocalOffset=<-6 0 0>}]
   }
   n25 = Anim Dynamics - Bone: dyn_tassel_mid!(ComponentPose=@n24, Alpha=1, GravityOverride:Vector="0,0,0", ExternalForce="0,0,0") -> (Pose) {
    angularDampingOverride=0.9999
    angularSpringConstant=500
    bAngularSpring=true
    bGravityOverrideInSimSpace=true
    bOverrideAngularDamping=true
    bUseGravityOverride=true
    bUsePlanarLimit=false
    bUseSphericalLimits=true
    simulationSpace=BoneRelative
    alphaBoolBlend {blendOption=Linear}
    boundBone {boneName=dyn_tassel_mid}
    componentAppliedLinearAccClamp=<3500 3500 3500>
    componentLinearAccScale=<1.2 1.2 1.2>
    componentLinearVelScale=<1.2 1.2 1.2>
    physicsBodyDefinitions [{
      collisionType=CustomSphere
      sphereCollisionRadius=2
      boundBone {boneName=dyn_tassel_mid}
      boxExtents=<15 30 30>
      constraintSetup {angularLimitsMax=<1 20 15>, angularLimitsMin=<-1 -45 -15>, angularTarget=<0 0 0>}
      localJointOffset=<4 0 0>
     }]
    relativeSpaceBone {boneName=root}
    sphericalLimits [{limitRadius=10, drivingBone {boneName=thigh_l}, sphereLocalOffset=<-6 0 0>}]
   }
   n26 = Anim Dynamics - Bone: dyn_tassel_bk!(ComponentPose=@n25, Alpha=1, GravityOverride:Vector="0,0,0", ExternalForce="0,0,0") -> (Pose) {
    angularDampingOverride=0.9999
    angularSpringConstant=500
    bAngularSpring=true
    bGravityOverrideInSimSpace=true
    bOverrideAngularDamping=true
    bUseGravityOverride=true
    bUsePlanarLimit=false
    bUseSphericalLimits=true
    simulationSpace=BoneRelative
    alphaBoolBlend {blendOption=Linear}
    boundBone {boneName=dyn_tassel_bk}
    componentAppliedLinearAccClamp=<3500 3500 3500>
    componentLinearAccScale=<1.2 1.2 1.2>
    componentLinearVelScale=<1.2 1.2 1.2>
    physicsBodyDefinitions [{
      collisionType=CustomSphere
      sphereCollisionRadius=2
      boundBone {boneName=dyn_tassel_bk}
      boxExtents=<15 30 30>
      constraintSetup {angularLimitsMax=<1 5 15>, angularLimitsMin=<-1 -35 -15>, angularTarget=<0 0 0>}
      localJointOffset=<4 0 0>
     }]
    relativeSpaceBone {boneName=root}
    sphericalLimits [{limitRadius=10, drivingBone {boneName=thigh_l}, sphereLocalOffset=<-6 0 0>}]
   }
   n27 = Anim Dynamics - Bone: dyn_bag_bk_01(ComponentPose=@n26, Alpha=0.4, ExternalForce="0,0,0") {
    angularDampingOverride=0.9999
    angularSpringConstant=500
    bAngularSpring=true
    bGravityOverrideInSimSpace=true
    bOverrideAngularDamping=true
    simulationSpace=BoneRelative
    alphaBoolBlend {blendOption=Linear}
    boundBone {boneName=dyn_bag_bk_01}
    componentAppliedLinearAccClamp=<3500 3500 3500>
    componentLinearAccScale=<1.2 1.2 1.2>
    componentLinearVelScale=<1.2 1.2 1.2>
    physicsBodyDefinitions [{
      sphereCollisionRadius=0
      boundBone {boneName=dyn_bag_bk_01}
      boxExtents=<15 30 30>
      constraintSetup {angularLimitsMax=<1 5 15>, angularLimitsMin=<-1 -35 -30>, angularTarget=<0 0 0>}
      localJointOffset=<4 0 0>
     }]
    relativeSpaceBone {boneName=root}
   }
   n28 = Anim Dynamics - Bone: dyn_band_top_01(ComponentPose=@n27, Alpha=0.7, ExternalForce="0,0,0") {
    angularDampingOverride=1
    bOverrideAngularDamping=true
    bUseGravityOverride=true
    bUsePlanarLimit=false
    bUseSphericalLimits=true
    simulationSpace=BoneRelative
    alphaBoolBlend {blendOption=Linear}
    boundBone {boneName=dyn_band_top_01}
    componentAppliedLinearAccClamp=<3000 3000 3000>
    componentLinearAccScale=<5 5 8>
    componentLinearVelScale=<5 5 8>
    gravityOverride {z=-980}
    physicsBodyDefinitions [{
      collisionType=CustomSphere
      sphereCollisionRadius=1
      boundBone {boneName=dyn_band_top_01}
      boxExtents=<25 25 25>
      constraintSetup {angularLimitsMax=<3 45 2>, angularLimitsMin=<-3 -45 -85>, angularTarget=<0 0 0>}
      localJointOffset=<2 0 0>
     }]
    relativeSpaceBone {boneName=root}
    sphericalLimits [
     {limitRadius=17, drivingBone {boneName=spine_04}, sphereLocalOffset=<0 0 1.5>}
     {limitRadius=5, drivingBone {boneName=deform_elbow_out_r}, sphereLocalOffset=<0 0 0>}
     {limitRadius=2, drivingBone {boneName=dyn_belt_end_r}, sphereLocalOffset=<0 0 0>}
    ]
   }
   n29 = Anim Dynamics - Bone: dyn_band_top_02(ComponentPose=@n28, Alpha=1, ExternalForce="0,0,0") {
    angularDampingOverride=0.9999
    bOverrideAngularDamping=true
    bUseGravityOverride=true
    bUsePlanarLimit=false
    bUseSphericalLimits=true
    simulationSpace=BoneRelative
    alphaBoolBlend {blendOption=Linear}
    boundBone {boneName=dyn_band_top_02}
    componentAppliedLinearAccClamp=<3000 3000 3000>
    componentLinearAccScale=<5 5 8>
    componentLinearVelScale=<5 5 8>
    gravityOverride {z=-980}
    physicsBodyDefinitions [{
      collisionType=CustomSphere
      sphereCollisionRadius=1
      boundBone {boneName=dyn_band_top_02}
      boxExtents=<25 25 25>
      constraintSetup {angularLimitsMax=<3 45 55>, angularLimitsMin=<-3 -45 -55>, angularTarget=<0 0 0>}
      localJointOffset=<10 0 0>
     }]
    relativeSpaceBone {boneName=root}
    sphericalLimits [
     {limitRadius=17, drivingBone {boneName=spine_04}, sphereLocalOffset=<-3 0 1.5>}
     {limitRadius=5, drivingBone {boneName=deform_elbow_out_r}, sphereLocalOffset=<8 -8 0>}
     {limitRadius=2, drivingBone {boneName=dyn_belt_end_r}, sphereLocalOffset=<0 0 0>}
    ]
   }
   n30 = Anim Dynamics - Bone: dyn_band_top_03(ComponentPose=@n29, Alpha=1, ExternalForce="0,0,0") {
    angularDampingOverride=0.999
    bOverrideAngularDamping=true
    bUseGravityOverride=true
    bUsePlanarLimit=false
    bUseSphericalLimits=true
    simulationSpace=BoneRelative
    alphaBoolBlend {blendOption=Linear}
    boundBone {boneName=dyn_band_top_03}
    componentAppliedLinearAccClamp=<3000 3000 3000>
    componentLinearAccScale=<5 5 8>
    componentLinearVelScale=<5 5 8>
    gravityOverride {z=-980}
    physicsBodyDefinitions [{
      collisionType=CustomSphere
      sphereCollisionRadius=1
      boundBone {boneName=dyn_band_top_03}
      boxExtents=<25 25 25>
      constraintSetup {angularLimitsMax=<3 55 55>, angularLimitsMin=<-3 -55 -55>, angularTarget=<0 0 0>}
      localJointOffset=<10 0 0>
     }]
    relativeSpaceBone {boneName=root}
    sphericalLimits [
     {limitRadius=17, drivingBone {boneName=spine_04}, sphereLocalOffset=<-6 0 1.5>}
     {limitRadius=8, drivingBone {boneName=deform_elbow_out_r}, sphereLocalOffset=<6 -8 0>}
     {limitRadius=3, drivingBone {boneName=dyn_belt_end_r}, sphereLocalOffset=<0 0 0>}
    ]
   }
   n31 = Anim Dynamics - Bone: dyn_band_bot_01(ComponentPose=@n30, Alpha=0.6, ExternalForce="0,0,0") {
    angularDampingOverride=1
    bOverrideAngularDamping=true
    bUseGravityOverride=true
    bUsePlanarLimit=false
    bUseSphericalLimits=true
    simulationSpace=BoneRelative
    alphaBoolBlend {blendOption=Linear}
    boundBone {boneName=dyn_band_bot_01}
    componentAppliedLinearAccClamp=<3000 3000 3000>
    componentLinearAccScale=<5 5 8>
    componentLinearVelScale=<5 5 8>
    gravityOverride {z=-980}
    physicsBodyDefinitions [{
      collisionType=CustomSphere
      sphereCollisionRadius=1
      boundBone {boneName=dyn_band_bot_01}
      boxExtents=<20 25 25>
      constraintSetup {angularLimitsMax=<3 45 2>, angularLimitsMin=<-3 -45 -85>, angularTarget=<0 0 0>}
      localJointOffset=<2 0 0>
     }]
    relativeSpaceBone {boneName=root}
    sphericalLimits [
     {limitRadius=17, drivingBone {boneName=spine_04}, sphereLocalOffset=<0 0 1.5>}
     {limitRadius=5, drivingBone {boneName=deform_elbow_out_r}, sphereLocalOffset=<0 0 0>}
     {limitRadius=2, drivingBone {boneName=dyn_belt_end_r}, sphereLocalOffset=<0 0 0>}
    ]
   }
   n32 = Anim Dynamics - Bone: dyn_band_bot_02(ComponentPose=@n31, Alpha=1, ExternalForce="0,0,0") {
    angularDampingOverride=0.9999
    bOverrideAngularDamping=true
    bUseGravityOverride=true
    bUsePlanarLimit=false
    bUseSphericalLimits=true
    simulationSpace=BoneRelative
    alphaBoolBlend {blendOption=Linear}
    boundBone {boneName=dyn_band_bot_02}
    componentAppliedLinearAccClamp=<3000 3000 3000>
    componentLinearAccScale=<5 5 8>
    componentLinearVelScale=<5 5 8>
    gravityOverride {z=-980}
    physicsBodyDefinitions [{
      collisionType=CustomSphere
      sphereCollisionRadius=1
      boundBone {boneName=dyn_band_bot_02}
      boxExtents=<25 25 25>
      constraintSetup {angularLimitsMax=<3 55 55>, angularLimitsMin=<-3 -55 -55>, angularTarget=<0 0 0>}
      localJointOffset=<10 0 0>
     }]
    relativeSpaceBone {boneName=root}
    sphericalLimits [
     {limitRadius=17, drivingBone {boneName=spine_04}, sphereLocalOffset=<-3 0 1.5>}
     {limitRadius=5, drivingBone {boneName=deform_elbow_out_r}, sphereLocalOffset=<8 -8 0>}
     {limitRadius=2, drivingBone {boneName=dyn_belt_end_r}, sphereLocalOffset=<0 0 0>}
    ]
   }
   n33 = Anim Dynamics - Bone: dyn_band_bot_03(ComponentPose=@n32, Alpha=1, ExternalForce="0,0,0") {
    angularDampingOverride=0.999
    bOverrideAngularDamping=true
    bUseGravityOverride=true
    bUsePlanarLimit=false
    bUseSphericalLimits=true
    simulationSpace=BoneRelative
    alphaBoolBlend {blendOption=Linear}
    boundBone {boneName=dyn_band_bot_03}
    componentAppliedLinearAccClamp=<3000 3000 3000>
    componentLinearAccScale=<5 5 8>
    componentLinearVelScale=<5 5 8>
    gravityOverride {z=-980}
    physicsBodyDefinitions [{
      collisionType=CustomSphere
      sphereCollisionRadius=1
      boundBone {boneName=dyn_band_bot_03}
      boxExtents=<20 20 20>
      constraintSetup {angularLimitsMax=<3 55 55>, angularLimitsMin=<-3 -55 -55>, angularTarget=<0 0 0>}
      localJointOffset=<15 0 0>
     }]
    relativeSpaceBone {boneName=root}
    sphericalLimits [
     {limitRadius=17, drivingBone {boneName=spine_04}, sphereLocalOffset=<-6 0 1.5>}
     {limitRadius=8, drivingBone {boneName=deform_elbow_out_r}, sphereLocalOffset=<6 -8 0>}
     {limitRadius=3, drivingBone {boneName=dyn_belt_end_r}, sphereLocalOffset=<0 0 0>}
    ]
   }
   n34 = "Transform (Modify) Bone - Bone: dyn_band_top_01"!(ComponentPose=@n33, Alpha=1, Rotation="0.000,-50.000,-0.000") -> (Pose) {
    rotationMode=BMM_Additive
    rotationSpace=BCS_BoneSpace
    alphaBoolBlend {blendOption=Linear}
    alphaScaleBiasClamp {bInterpResult=true, interpSpeedDecreasing=5, interpSpeedIncreasing=7}
    boneToModify {boneName=dyn_band_top_01}
    rotation=<rot 0 -50 0>
   }
   n35 = "Transform (Modify) Bone - Bone: dyn_band_bot_01"!(ComponentPose=@n34, Alpha=1, Rotation="0.000,-5.000,0.000") -> (Pose) {
    rotationMode=BMM_Additive
    rotationSpace=BCS_BoneSpace
    alphaBoolBlend {blendOption=Linear}
    alphaScaleBiasClamp {bInterpResult=true, interpSpeedDecreasing=5, interpSpeedIncreasing=7}
    boneToModify {boneName=dyn_band_bot_01}
    rotation=<rot 0 -5 0>
   }
   n36 = Look At - Bone: dyn_band_top_02(ComponentPose=@n35, Alpha=0.5) {
    alphaBoolBlend {blendOption=Linear}
    alphaScaleBiasClamp {bClampResult=true, bInterpResult=true, clampMax=0.2, interpSpeedDecreasing=5, interpSpeedIncreasing=7}
    boneToModify {boneName=dyn_band_top_02}
    lookAtLocation=<-52.5301 -56.3448 16.562>
    lookAtTarget {boneReference {boneName=spine_05}}
    lookAt_Axis {axis=<1 0>}
   }
   n37 = Anim Dynamics - Bone: dyn_boot_l_01(ComponentPose=@n36, Alpha=1, ExternalForce="0,0,0") {
    angularDampingOverride=0.9999
    bOverrideAngularDamping=true
    bUseGravityOverride=true
    bUsePlanarLimit=false
    bUseSphericalLimits=true
    simulationSpace=BoneRelative
    alphaBoolBlend {blendOption=Linear}
    boundBone {boneName=dyn_boot_l_01}
    componentAppliedLinearAccClamp=<3000 3000 3000>
    componentLinearAccScale=<5 5 8>
    componentLinearVelScale=<5 5 8>
    gravityOverride {z=-980}
    physicsBodyDefinitions [{
      collisionType=CustomSphere
      sphereCollisionRadius=1
      boundBone {boneName=dyn_boot_l_01}
      boxExtents=<25 25 25>
      constraintSetup {angularLimitsMax=<3 30 55>, angularLimitsMin=<-3 -30 -30>, angularTarget=<0 0 0>}
      localJointOffset=<5 0 0>
     }]
    relativeSpaceBone {boneName=root}
   }
   n38 = Anim Dynamics - Bone: dyn_boot_l_02(ComponentPose=@n37, Alpha=1, ExternalForce="0,0,0") {
    angularDampingOverride=0.999
    bOverrideAngularDamping=true
    bUseGravityOverride=true
    bUsePlanarLimit=false
    bUseSphericalLimits=true
    simulationSpace=BoneRelative
    alphaBoolBlend {blendOption=Linear}
    boundBone {boneName=dyn_boot_l_02}
    componentAppliedLinearAccClamp=<3000 3000 3000>
    componentLinearAccScale=<5 5 8>
    componentLinearVelScale=<5 5 8>
    gravityOverride {z=-980}
    physicsBodyDefinitions [{
      collisionType=CustomSphere
      sphereCollisionRadius=1
      boundBone {boneName=dyn_boot_l_02}
      boxExtents=<25 25 25>
      constraintSetup {angularLimitsMax=<3 30 55>, angularLimitsMin=<-3 -30 -30>, angularTarget=<0 0 0>}
      localJointOffset=<10 0 0>
     }]
    relativeSpaceBone {boneName=root}
   }
   n39 = Anim Dynamics - Bone: dyn_boot_r_01(ComponentPose=@n38, Alpha=1, ExternalForce="0,0,0") {
    angularDampingOverride=0.9999
    bOverrideAngularDamping=true
    bUseGravityOverride=true
    bUsePlanarLimit=false
    bUseSphericalLimits=true
    simulationSpace=BoneRelative
    alphaBoolBlend {blendOption=Linear}
    boundBone {boneName=dyn_boot_r_01}
    componentAppliedLinearAccClamp=<3000 3000 3000>
    componentLinearAccScale=<5 5 8>
    componentLinearVelScale=<5 5 8>
    gravityOverride {z=-980}
    physicsBodyDefinitions [{
      collisionType=CustomSphere
      sphereCollisionRadius=1
      boundBone {boneName=dyn_boot_r_01}
      boxExtents=<25 25 25>
      constraintSetup {angularLimitsMax=<3 30 55>, angularLimitsMin=<-3 -30 -30>, angularTarget=<0 0 0>}
      localJointOffset=<8 0 0>
     }]
    relativeSpaceBone {boneName=root}
   }
   n40 = Anim Dynamics - Bone: dyn_boot_r_02(ComponentPose=@n39, Alpha=1, ExternalForce="0,0,0") {
    angularDampingOverride=0.999
    bOverrideAngularDamping=true
    bUseGravityOverride=true
    bUsePlanarLimit=false
    bUseSphericalLimits=true
    simulationSpace=BoneRelative
    alphaBoolBlend {blendOption=Linear}
    boundBone {boneName=dyn_boot_r_02}
    componentAppliedLinearAccClamp=<3000 3000 3000>
    componentLinearAccScale=<5 5 8>
    componentLinearVelScale=<5 5 8>
    gravityOverride {z=-980}
    physicsBodyDefinitions [{
      collisionType=CustomSphere
      sphereCollisionRadius=1
      boundBone {boneName=dyn_boot_r_02}
      boxExtents=<20 20 20>
      constraintSetup {angularLimitsMax=<3 30 55>, angularLimitsMin=<-3 -30 -30>, angularTarget=<0 0 0>}
      localJointOffset=<15 0 0>
     }]
    relativeSpaceBone {boneName=root}
   }
   n41 = Control Rig!(
    Source=Component To Local(ComponentPose=@n40)
    Alpha
    Jumps:double=Get Jumps() -> (Jumps:double)
    DBNO:double=Get DBNO() -> (DBNO:double)
    Target:double=Get ADS() -> (ADS:double)
    Crouch:double=Get Crouch() -> (Crouch:double)
   ) -> (Pose) {
    alphaBoolBlend {blendOption=Linear}
    controlRigAssetReference {blueprintRigClass=&/F_MED_HighBeam_CtrlRig.F_MED_HighBeam_CtrlRig_C}
   }
   n42 = "Transform (Modify) Bone - Bone: dyn_dagger_02"!(ComponentPose=Local To Component(LocalPose=@n41), Alpha=1, Rotation="4.752,-45.172,-1.489") -> (Pose) {
    rotationMode=BMM_Replace
    rotationSpace=BCS_BoneSpace
    alphaBoolBlend {blendOption=Linear}
    alphaScaleBiasClamp {interpSpeedDecreasing=5, interpSpeedIncreasing=7}
    boneToModify {boneName=dyn_dagger_02}
    rotation=<rot 4.752 -45.172 -1.489>
   }
   n43 = "Transform (Modify) Bone - Bone: dyn_belt_MSTR"!(ComponentPose=@n42, Alpha=1, Rotation="4.351,-59.197,17.627") -> (Pose) {
    rotationMode=BMM_Replace
    rotationSpace=BCS_BoneSpace
    alphaBoolBlend {blendOption=Linear}
    alphaScaleBiasClamp {interpSpeedDecreasing=5, interpSpeedIncreasing=7}
    boneToModify {boneName=dyn_belt_MSTR}
    rotation=<rot 4.351 -59.197 17.627>
   }
   n44 = "Transform (Modify) Bone - Bone: dyn_dagger_02"!(ComponentPose=@n43, Alpha=1, Rotation="5.063,24.953,0.298", Scale:Vector="0.9,0.9,0.9") -> (Pose) {
    rotationMode=BMM_Replace
    rotationSpace=BCS_BoneSpace
    scaleMode=BMM_Replace
    scaleSpace=BCS_BoneSpace
    alphaBoolBlend {blendOption=Linear}
    alphaScaleBiasClamp {interpSpeedDecreasing=5, interpSpeedIncreasing=7}
    boneToModify {boneName=dyn_dagger_02}
    rotation=<rot 5.063 24.953 0.298>
    scale=<0.9 0.9 0.9>
   }
   n45 = Get GravityOverride()
   n46 = RigidBody(ComponentPose=@n44, Alpha=Get Skydive Off() -> (Skydive Off:double), OverrideWorldGravity=@n45) {
    bClampLinearTranslationLimitToRefPose=true
    bOverrideWorldGravity=true
    overridePhysicsAsset=&F_MED_HighBeam_Physics
    simulationSpace=BaseBoneSpace
    alphaBoolBlend {blendOption=Linear}
    baseBoneRef {boneName=pelvis}
    componentAppliedLinearAccClamp=<3000 3000 3000>
    componentLinearAccScale=<0.5 0.3 0.3>
    componentLinearVelScale=<0.5 0.3 0.3>
    overrideWorldGravity {z=-980}
   }
   n47 = RigidBody!(ComponentPose=@n46, bAlphaBoolEnabled:bool, OverrideWorldGravity=@n45, ExternalForce) -> (Pose) {
    alphaInputType=Bool
    bClampLinearTranslationLimitToRefPose=true
    bOverrideWorldGravity=true
    overridePhysicsAsset=&F_MED_HighBeam_Physics
    simulationSpace=BaseBoneSpace
    alphaBoolBlend {blendInTime=0.1, blendOption=Linear, blendOutTime=0.1}
    baseBoneRef {boneName=root}
    componentAppliedLinearAccClamp=<3000 3000 3000>
    componentLinearAccScale=<0.3 0.3 0.4>
    componentLinearVelScale=<0.3 0.3 0.4>
    simSpaceSettings {
     maxAngularAcceleration=1000
     maxAngularVelocity=3
     maxLinearAcceleration=700
     maxLinearVelocity=500
     velocityScaleZ=0
     externalLinearDragV=<0.2 0.2>
    }
   }
   n48 = "Transform (Modify) Bone - Bone: dyn_bow_MSTR"!(ComponentPose=@n47, Alpha=1, Translation="1,0,0", Rotation="0,-2,0") -> (Pose) {
    rotationMode=BMM_Additive
    rotationSpace=BCS_BoneSpace
    translationMode=BMM_Replace
    translationSpace=BCS_BoneSpace
    alphaBoolBlend {blendOption=Linear}
    alphaScaleBiasClamp {bInterpResult=true, interpSpeedDecreasing=5, interpSpeedIncreasing=7}
    boneToModify {boneName=dyn_bow_MSTR}
    rotation {yaw=-2}
    translation {x=1}
   }
   n49 = "Transform (Modify) Bone - Bone: dyn_tie_l_01"!(ComponentPose=@n48, Alpha=1, Translation="1,0,0", Rotation="0.000,15.000,-0.000") -> (Pose) {
    rotationMode=BMM_Additive
    rotationSpace=BCS_BoneSpace
    translationMode=BMM_Replace
    translationSpace=BCS_BoneSpace
    alphaBoolBlend {blendOption=Linear}
    alphaScaleBiasClamp {bInterpResult=true, interpSpeedDecreasing=5, interpSpeedIncreasing=7}
    boneToModify {boneName=dyn_tie_l_01}
    rotation=<rot 0 15 0>
    translation {x=1}
   }
   n50 = "Transform (Modify) Bone - Bone: dyn_tie_r_01"!(ComponentPose=@n49, Alpha=1, Translation="1,0,0", Rotation="0.000,-10.000,0.000") -> (Pose) {
    rotationMode=BMM_Additive
    rotationSpace=BCS_BoneSpace
    translationMode=BMM_Replace
    translationSpace=BCS_BoneSpace
    alphaBoolBlend {blendOption=Linear}
    alphaScaleBiasClamp {bInterpResult=true, interpSpeedDecreasing=5, interpSpeedIncreasing=7}
    boneToModify {boneName=dyn_tie_r_01}
    rotation=<rot 0 -10 0>
    translation {x=1}
   }
   n51 = "Constraint - dyn_bag_bk_02 (PPP)"!(
    ComponentPose=@n50
    bAlphaBoolEnabled:bool=True
    ConstraintWeights_0=0.7192
    ConstraintWeights_1:float=0.0203
    ConstraintWeights_2:float=0.2605
   ) -> (Pose) {
    alphaInputType=Bool
    alphaBoolBlend {blendOption=Linear}
    boneToModify {boneName=dyn_bag_bk_02}
    constraintSetup [{targetBone {boneName=pelvis}}, {targetBone {boneName=thigh_twist_01_l}}, {targetBone {boneName=deform_glute_l}}]
    constraintWeights [1,1,1]
   }
   n52 = "Constraint - dyn_band_top_02 (PPP)"!(ComponentPose=@n51, Alpha=1, ConstraintWeights_0=0.2176, ConstraintWeights_1:float=0.2216, ConstraintWeights_2:float=0.5607) -> (Pose) {
    alphaBoolBlend {blendOption=Linear}
    boneToModify {boneName=dyn_band_top_02}
    constraintSetup [{targetBone {boneName=upperarm_r}}, {targetBone {boneName=upperarm_twist_02_r}}, {targetBone {boneName=upperarm_twist_01_r}}]
    constraintWeights [1,1,1]
   }
   n53 = "Constraint - dyn_band_bot_02 (PPP)"!(ComponentPose=@n52, Alpha=1, ConstraintWeights_0=0.2176, ConstraintWeights_1:float=0.2216, ConstraintWeights_2:float=0.5607) -> (Pose) {
    alphaBoolBlend {blendOption=Linear}
    boneToModify {boneName=dyn_band_bot_02}
    constraintSetup [{targetBone {boneName=upperarm_r}}, {targetBone {boneName=upperarm_twist_02_r}}, {targetBone {boneName=upperarm_twist_01_r}}]
    constraintWeights [1,1,1]
   }
   n54 = "Transform (Modify) Bone - Bone: dyn_band_top_01"!(ComponentPose=@n53, Alpha=1, Translation:Vector="1,0,0", Rotation="-49.920,12.193,-18.130") -> (Pose) {
    rotationMode=BMM_Replace
    rotationSpace=BCS_BoneSpace
    translationMode=BMM_Replace
    translationSpace=BCS_BoneSpace
    alphaBoolBlend {blendOption=Linear}
    alphaScaleBiasClamp {bInterpResult=true, interpSpeedDecreasing=5, interpSpeedIncreasing=7}
    boneToModify {boneName=dyn_band_top_01}
    rotation=<rot -49.92 12.193 -18.13>
    translation {x=1}
   }
   n55 = "Transform (Modify) Bone - Bone: dyn_band_bot_01"!(ComponentPose=@n54, Alpha=1, Translation:Vector="1,0,0", Rotation="-39.133,10.506,-10.597") -> (Pose) {
    rotationMode=BMM_Replace
    rotationSpace=BCS_BoneSpace
    translationMode=BMM_Replace
    translationSpace=BCS_BoneSpace
    alphaBoolBlend {blendOption=Linear}
    alphaScaleBiasClamp {bInterpResult=true, interpSpeedDecreasing=5, interpSpeedIncreasing=7}
    boneToModify {boneName=dyn_band_bot_01}
    rotation=<rot -39.133 10.5061 -10.597>
    translation {x=1}
   }
   n56 = "Constraint - dyn_tie_l_01 (PPPPP)"!(
    ComponentPose=@n55
    Alpha=1
    ConstraintWeights_0=0.6005
    ConstraintWeights_1:float=0.3784
    ConstraintWeights_2:float=0.001
    ConstraintWeights_3:float=0.0021
    ConstraintWeights_4:float=0.018
   ) -> (Pose) {
    alphaBoolBlend {blendOption=Linear}
    alphaScaleBiasClamp {bClampResult=true, bInterpResult=true, interpSpeedDecreasing=5, interpSpeedIncreasing=7}
    boneToModify {boneName=dyn_tie_l_01}
    constraintSetup [
     {targetBone {boneName=spine_04}}
     {targetBone {boneName=spine_05}}
     {targetBone {boneName=clavicle_r}}
     {targetBone {boneName=deform_pec_l}}
     {targetBone {boneName=deform_pec_r}}
    ]
    constraintWeights [1,1,1,1,1]
   }
   n57 = "Constraint - dyn_tie_r_02 (PPPPP)"!(
    ComponentPose=@n56
    Alpha=1
    ConstraintWeights_0=0.5854
    ConstraintWeights_1:float=0.3745
    ConstraintWeights_2:float=0.0025
    ConstraintWeights_3:float=0.0016
    ConstraintWeights_4:float=0.0359
   ) -> (Pose) {
    alphaBoolBlend {blendOption=Linear}
    alphaScaleBiasClamp {bClampResult=true, bInterpResult=true, interpSpeedDecreasing=5, interpSpeedIncreasing=7}
    boneToModify {boneName=dyn_tie_r_02}
    constraintSetup [
     {targetBone {boneName=spine_04}}
     {targetBone {boneName=spine_05}}
     {targetBone {boneName=clavicle_r}}
     {targetBone {boneName=deform_pec_l}}
     {targetBone {boneName=deform_pec_r}}
    ]
    constraintWeights [1,1,1,1,1]
   }
   Output Pose(Result=Component To Local(ComponentPose=@n57)) {layerGroup=None}
  ]}
 EventGraph {nodes [
   n1 = Event BlueprintUpdateAnimation()
   > n2 = Sequence!(execute) -> (then_0, then_1, then_2:Exec, then_3:Exec)
   > n3 = Gravity Override State Function DEPRECATED(CharacterPartAnimInstance=Self-Reference(), GameTimeInSeconds=GetGameTimeInSeconds(), GravityOverrideParams=Get Gravity Override Params())
   > Set GravityOverride(GravityOverride=@n3)
   n4 = CheckForBackbling()
   Set Cape(execute=@n2.then_1, Cape:bool=OR Boolean(A=@n4.bIsBackblingACape, B=@n4.bIsBackblingLong)) -> (Output_Get:bool)
   > Set Skydive Off(Skydive Off:double=SelectFloat(A=0, B=1, bPickA=Get bIsSkyDiving())) -> (Output_Get:double)
   n5 = NOT Boolean(A=Get bIsSkyDiving())
   n6 = OR Boolean!(A=Get bIsJumping(), B=Get bIsFalling(), C:bool) -> (ReturnValue)
   > Set Jumps(Jumps:double=SelectFloat(A=1, B=0, bPickA=AND Boolean(A=@n6, B=@n5))) -> (Output_Get:double)
   > Set DBNO(DBNO:double=SelectFloat(A=1, B=0, bPickA=Get bIsDBNO())) -> (Output_Get:double)
   > Set ADS(ADS:double=SelectFloat(A=1, B=0, bPickA=Get bIsTargeting())) -> (Output_Get:double)
   > Set Crouch(Crouch:double=SelectFloat(A=1, B=0, bPickA=Get bIsCrouching())) -> (Output_Get:double)
   n7 = PlayerMovement(execute=@n2.then_2, AnimInstance=Self-Reference(), XDotProductCoefficient=100, YDotProductCoefficient=100, ZDotProductCoefficient=100)
   n8 = BreakVector(InVec=MakeVector(X=@n7, Y=@n7.YMovement))
   n9 = "float < float"(A=@n8, B=-50)
   > Set Left(Left:double=ClampAxis!(Angle:float="To Float (Boolean)"(InBool=@n9)) -> (ReturnValue:float)) -> (Output_Get:double)
   n10 = "float > float"(A=@n8, B=50)
   > Set Right(Right:double=ClampAxis(Angle="To Float (Boolean)"(InBool=@n10))) -> (Output_Get:double)
   n11 = "float > float"(A=@n8.Y, B=40)
   > Set Forward(Forward:double=ClampAxis(Angle="To Float (Boolean)"(InBool=@n11))) -> (Output_Get:double)
   Is Valid(exec=@n2.then_3, InputObject=Get PlayerPawn() -> (PlayerPawn:Fort Player Pawn Athena Object Reference))
   > n12 = Branch(Condition=Get bIsInFrontEnd())
   Set WPO Globals Function(
    execute=@n12.else
    Material=Get Mat Instance Dynamic()
    Overlay Material=Get Overlay Mat Instance Dynamic()
    Skeletal Mesh=Get SkelMesh()
    "Local Noise Transform  Location"=Get Local Noise Transform Location() -> (Local Noise Transform Location:Vector)
    Local Noise Offset=Get Local Noise Offset()
   )
   > Set WPO Regional Function!(
    execute
    Material:Material Instance Dynamic Object Reference=Get Mat Instance Dynamic()
    Overlay Material:Material Instance Dynamic Object Reference=Get Overlay Mat Instance Dynamic()
    Skeletal Mesh:Skeletal Mesh Component Object Reference=Get SkelMesh()
    Velocity Transform:Name=hand_r
    Velocity Interp Speed:double=Get Velocity Interp Speed()
    Collision Transform:Name=upperarm_r
    Collision Transform Axis:EAxis Enum
    Invert Collision Transform Axis:bool=false
    Delta Time X:double=@n1.DeltaTimeX
    Previous Velocity Transform Location:"Vector (by ref)"=Get R1 Velocity Transform Location()
    Previous Velocity Transform Velocity:"Vector (by ref)"=Get R1 Velocity Transform Velocity()
    Region:int=1
   ) -> (then)
   > Set WPO Regional Function(
    Material=Get Mat Instance Dynamic()
    Overlay Material=Get Overlay Mat Instance Dynamic()
    Skeletal Mesh=Get SkelMesh()
    Velocity Transform=hand_l
    Velocity Interp Speed=Get Velocity Interp Speed()
    Collision Transform=upperarm_l
    Invert Collision Transform Axis=true
    Delta Time X=@n1.DeltaTimeX
    Previous Velocity Transform Location=Get R2 Velocity Transform Location()
    Previous Velocity Transform Velocity=Get R2 Velocity Transform Velocity()
    Region=2
   )
   > n13 = Branch(Condition=Get bIsSkyDiving())
   > n14 = Set WPO Pho Wind Intensity(WPO Pho Wind Intensity=1)
   SetScalarParameterValue(
    execute=[@n15.then, @n14.then]
    self=[Get Mat Instance Dynamic(), Get Overlay Mat Instance Dynamic()]
    ParameterName=ClothSim_WindStrength
    Value=Get WPO Pho Wind Intensity()
   )
   n15 = Set WPO Pho Wind Intensity(execute=@n13.else, WPO Pho Wind Intensity=0)
   Event BlueprintInitializeAnimation()
   > Set SkelMesh(SkelMesh=GetOwnerSkeletalMeshComponent())
   n16 = TryGetPawnOwner()
   > Is Valid(InputObject=@n16)
   > n17 = Cast To FortPlayerPawnAthena(Object=@n16)
   > Set PlayerPawn(PlayerPawn:Fort Player Pawn Athena Object Reference=@n17) -> (Output_Get:Fort Player Pawn Athena Object Reference)
   Event OnCharacterCustomizationLoaded()
   > n18 = Set SkelMesh(SkelMesh=GetPartSkeletalMeshComponent())
   > n19 = Get Dynamic Material Instance(SkeletalMeshComponent=@n18)
   > Set Mat Instance Dynamic(Mat Instance Dynamic=@n19)
   > n20 = GetOrCreateDynamicOverlayMaterial(MeshComponent=@n18, MaterialName=None)
   > Set Overlay Mat Instance Dynamic(Overlay Mat Instance Dynamic=@n20)
   Get bIsSkydiveFloating()
  ]}
]
variables [
 ADS {type=number}
 Cape {type=boolean}
 Crouch {type=number}
 DBNO {type=number}
 Forward {type=number}
 Gravity Override Params {type=GravityOverrideParamsStruct}
 GravityOverride {type=Vector}
 Jumps {type=number}
 Left {type=number}
 Local Noise Offset {type=Vector}
 Local Noise Transform Location {type=Vector}
 Mat Instance Dynamic {type=MaterialInstanceDynamic}
 Overlay Mat Instance Dynamic {type=MaterialInstanceDynamic}
 PlayerPawn {type=FortPlayerPawnAthena}
 R1 Velocity Transform Location {type=Vector}
 R1 Velocity Transform Velocity {type=Vector}
 R2 Velocity Transform Location {type=Vector}
 R2 Velocity Transform Velocity {type=Vector}
 Right {type=number}
 Self_0 {type=F_MED_HighBeam_AnimBP_C, default=null}
 SkelMesh {type=SkeletalMeshComponent}
 Skydive Off {type=number}
 Velocity Interp Speed {type=number}
 WPO Pho Wind Intensity {type=number}
]
