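# Control Rig for the HighBeam female character's FaceAcc part that offsets hair bones (side locks, bun, back strands) to follow head pose and player state.
# Five spherical pose readers measure head against neck_01, weighting chained ModifyTransforms steps, mostly smoothed through AlphaInterp.
# ADS scales one corrective, Crouching with an ADS clamp drives another, and Skydive alone drives the last.
# The second reader bypasses smoothing, leaving its AlphaInterp unused, and has debug draw enabled.
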ControlRigBlueprint /BRCosmetics/Characters/Player/Female/Medium/Bodies/F_MED_HighBeam/Meshes/Parts/F_MED_HighBeam_FaceAcc_CtrlRig
graphs [
 RigVMFunctionLibrary {links [], nodes []}
 RigVMModel {nodes [
   RigUnit_BeginExecution = &RigUnit_BeginExecution()
   n1 = &RigUnit_SphericalPoseReader(
    ActiveRegionScaleFactors=(PositiveWidth=0.5,NegativeWidth=0.2,PositiveHeight=1,NegativeHeight=1)
    ActiveRegionSize=0.2
    Debug=(bDrawDebug=false,DebugScale=25,DebugSegments=20,DebugThickness=0.25)
    DriverAxis=(X=1,Y=0,Z=0)
    DriverItem=(Type=Bone,Name="head")
    FalloffRegionScaleFactors=(PositiveWidth=0.28,NegativeWidth=0.1,PositiveHeight=0.4,NegativeHeight=0.2)
    FalloffSize=0.5
    OptionalParentItem=(Type=Bone,Name="neck_01")
    RotationOffset=(X=-90,Y=0,Z=55)
    ExecutePin=@RigUnit_BeginExecution.ExecutePin
   )
   AlphaInterp = &RigVMFunction_AlphaInterp(InterpSpeedDecreasing=3, InterpSpeedIncreasing=7, Scale=1, bClampResult=False, Value=@n1.OutputParam)
   > &RigUnit_ModifyTransforms(
    ItemToModify="((Item=(Type=Bone,Name=\"dyn_hair_side_l_mstr\"),Transform=(Rotation=(X=0,Y=0,Z=0,W=1),Translation=(X=0,Y=0,Z=0),Scale3D=(X=1,Y=1,Z=1))),(Item=(Type=Bone,Name=\"dyn_hair_side_r_mstr\"),Transform=(Rotation=(X=0,Y=0,Z=0,W=1),Translation=(X=0,Y=0,Z=0),Scale3D=(X= ...(truncated)"
    Weight=@AlphaInterp.Result
    ItemToModify.0.Transform.Rotation=&RigVMFunction_MathQuaternionFromEuler(Euler=(X=0,Y=0,Z=-5)).Result
    ItemToModify.1.Transform.Rotation=&RigVMFunction_MathQuaternionFromEuler(Euler=(X=0,Y=0,Z=5)).Result
    ItemToModify.2.Transform.Rotation=&RigVMFunction_MathQuaternionFromEuler(Euler=(X=0,Y=-5,Z=-5)).Result
    ItemToModify.3.Transform.Rotation=&RigVMFunction_MathQuaternionFromEuler(Euler=(X=0,Y=-5,Z=0)).Result
    ItemToModify.4.Transform.Rotation=&RigVMFunction_MathQuaternionFromEuler(Euler=(X=0,Y=-10,Z=0)).Result
   )
   > n2 = &RigUnit_SphericalPoseReader(
    ActiveRegionScaleFactors=(PositiveWidth=0.3,NegativeWidth=0.2,PositiveHeight=1,NegativeHeight=1)
    ActiveRegionSize=0.2
    Debug=(bDrawDebug=true,DebugScale=25,DebugSegments=20,DebugThickness=0.25)
    DriverAxis=(X=1,Y=0,Z=0)
    DriverItem=(Type=Bone,Name="head")
    FalloffRegionScaleFactors=(PositiveWidth=0.28,NegativeWidth=0.1,PositiveHeight=0.4,NegativeHeight=0.2)
    FalloffSize=0.5
    OptionalParentItem=(Type=Bone,Name="neck_01")
    RotationOffset=(X=-90,Y=0,Z=55)
   )
   > &RigUnit_ModifyTransforms(
    ItemToModify="((Item=(Type=Bone,Name=\"dyn_hair_bun\"),Transform=(Rotation=(X=0,Y=0,Z=0,W=1),Translation=(X=-0.05,Y=-0.2,Z=0),Scale3D=(X=1,Y=1,Z=1))),(Item=(Type=Bone,Name=\"dyn_hair_bk_r_01\"),Transform=(Rotation=(X=0,Y=0,Z=0,W=1),Translation=(X=-1,Y=0,Z=0),Scale3D=(X=1,Y= ...(truncated)"
    ItemToModify.0.Transform.Rotation=&RigVMFunction_MathQuaternionFromEuler(Euler=(X=0,Y=0,Z=-10)).Result
    ItemToModify.1.Transform.Rotation=&RigVMFunction_MathQuaternionFromEuler(Euler=(X=0,Y=-45,Z=0)).Result
    ItemToModify.2.Transform.Rotation=&RigVMFunction_MathQuaternionFromEuler(Euler=(X=0,Y=-20,Z=0)).Result
    ItemToModify.3.Transform.Rotation=&RigVMFunction_MathQuaternionFromEuler(Euler=(X=0,Y=-30,Z=-25)).Result
    ItemToModify.4.Transform.Rotation=&RigVMFunction_MathQuaternionFromEuler(Euler=(X=0,Y=-60,Z=0)).Result
    Weight=@n2.OutputParam
   )
   > n3 = &RigUnit_SphericalPoseReader(
    ActiveRegionScaleFactors=(PositiveWidth=0.25,NegativeWidth=0.1,PositiveHeight=1,NegativeHeight=0.5)
    ActiveRegionSize=0.2
    Debug=(bDrawDebug=false,DebugScale=25,DebugSegments=20,DebugThickness=0.25)
    DriverAxis=(X=1,Y=0,Z=0)
    DriverItem=(Type=Bone,Name="head")
    FalloffRegionScaleFactors=(PositiveWidth=0.1,NegativeWidth=0.05,PositiveHeight=0.2,NegativeHeight=0.2)
    FalloffSize=0.5
    OptionalParentItem=(Type=Bone,Name="neck_01")
    RotationOffset=(X=-90,Y=0,Z=55)
   )
   n4 = &RigVMFunction_AlphaInterp(InterpSpeedDecreasing=7, InterpSpeedIncreasing=3, bClampResult=False, Value=@n3.OutputParam, Scale=RigVMVariableNode() {variable=ADS}.Value)
   > &RigUnit_ModifyTransforms(
    ItemToModify="((Item=(Type=Bone,Name=\"dyn_hair_bun\"),Transform=(Rotation=(X=0,Y=0,Z=0,W=1),Translation=(X=0,Y=0,Z=0),Scale3D=(X=1,Y=1,Z=1))),(Item=(Type=Bone,Name=\"dyn_hair_bk_r_01\"),Transform=(Rotation=(X=0,Y=0,Z=0,W=1),Translation=(X=0,Y=0,Z=0),Scale3D=(X=1,Y=1,Z=1))) ...(truncated)"
    Weight=@n4.Result
    ItemToModify.4.Transform.Rotation=&RigVMFunction_MathQuaternionFromEuler(Euler=(X=0,Y=-20,Z=45)).Result
    ItemToModify.3.Transform.Rotation=&RigVMFunction_MathQuaternionFromEuler(Euler=(X=0,Y=-20,Z=0)).Result
    ItemToModify.1.Transform.Rotation=&RigVMFunction_MathQuaternionFromEuler(Euler=(X=0,Y=-30,Z=-10)).Result
    ItemToModify.0.Transform.Rotation=&RigVMFunction_MathQuaternionFromEuler(Euler=(X=0,Y=5,Z=-10)).Result
   )
   > n5 = &RigUnit_SphericalPoseReader(
    ActiveRegionScaleFactors=(PositiveWidth=0.25,NegativeWidth=0.1,PositiveHeight=1,NegativeHeight=0.5)
    ActiveRegionSize=0.2
    Debug=(bDrawDebug=false,DebugScale=25,DebugSegments=20,DebugThickness=0.25)
    DriverAxis=(X=1,Y=0,Z=0)
    DriverItem=(Type=Bone,Name="head")
    FalloffRegionScaleFactors=(PositiveWidth=0.1,NegativeWidth=0.05,PositiveHeight=0.2,NegativeHeight=0.2)
    FalloffSize=0.5
    OptionalParentItem=(Type=Bone,Name="neck_01")
    RotationOffset=(X=-90,Y=0,Z=55)
   )
   n6 = &RigVMFunction_AlphaInterp(
    InterpSpeedDecreasing=7
    InterpSpeedIncreasing=3
    bClampResult=true
    Value=@n5.OutputParam
    ClampMax=RigVMVariableNode() {variable=ADS}.Value
    Scale=RigVMVariableNode() {variable=Crouching}.Value
   )
   > &RigUnit_ModifyTransforms(
    ItemToModify="((Item=(Type=Bone,Name=\"dyn_hair_bun\"),Transform=(Rotation=(X=0,Y=0,Z=0,W=1),Translation=(X=0,Y=-2.5,Z=0),Scale3D=(X=1,Y=1,Z=1))),(Item=(Type=Bone,Name=\"dyn_hair_bk_r_01\"),Transform=(Rotation=(X=0,Y=0,Z=0,W=1),Translation=(X=0,Y=0,Z=0),Scale3D=(X=1,Y=1,Z=1 ...(truncated)"
    Weight=@n6.Result
    ItemToModify.4.Transform.Rotation=&RigVMFunction_MathQuaternionFromEuler(Euler=(X=0,Y=-60,Z=0)).Result
    ItemToModify.3.Transform.Rotation=&RigVMFunction_MathQuaternionFromEuler(Euler=(X=0,Y=-20,Z=0)).Result
    ItemToModify.1.Transform.Rotation=&RigVMFunction_MathQuaternionFromEuler(Euler=(X=0,Y=-20,Z=-10)).Result
    ItemToModify.0.Transform.Rotation=&RigVMFunction_MathQuaternionFromEuler(Euler=(X=0,Y=5,Z=-10)).Result
   )
   > n7 = &RigUnit_SphericalPoseReader(
    ActiveRegionScaleFactors=(PositiveWidth=0.1,NegativeWidth=0.8,PositiveHeight=1,NegativeHeight=1)
    ActiveRegionSize=0.2
    Debug=(bDrawDebug=false,DebugScale=25,DebugSegments=20,DebugThickness=0.25)
    DriverAxis=(X=1,Y=0,Z=0)
    DriverItem=(Type=Bone,Name="head")
    FalloffRegionScaleFactors=(PositiveWidth=0.2,NegativeWidth=0.3,PositiveHeight=0.5,NegativeHeight=0.5)
    FalloffSize=0.45
    OptionalParentItem=(Type=Bone,Name="neck_01")
    RotationOffset=(X=-90,Y=0,Z=130)
   )
   n8 = &RigVMFunction_AlphaInterp(InterpSpeedDecreasing=7, InterpSpeedIncreasing=5, Scale=1, bClampResult=False, Value=@n7.OutputParam)
   > &RigUnit_ModifyTransforms(
    ItemToModify="((Item=(Type=Bone,Name=\"dyn_hair_bun\"),Transform=(Rotation=(X=0,Y=0,Z=0,W=1),Translation=(X=0,Y=0,Z=0),Scale3D=(X=1,Y=1,Z=1))),(Item=(Type=Bone,Name=\"dyn_hair_bk_r_01\"),Transform=(Rotation=(X=0,Y=0,Z=0,W=1),Translation=(X=0,Y=0,Z=0),Scale3D=(X=1,Y=1,Z=1))) ...(truncated)"
    Weight=@n8.Result
    ItemToModify.4.Transform.Rotation=&RigVMFunction_MathQuaternionFromEuler(Euler=(X=0,Y=20,Z=0)).Result
    ItemToModify.3.Transform.Rotation=&RigVMFunction_MathQuaternionFromEuler(Euler=(X=0,Y=10,Z=0)).Result
    ItemToModify.2.Transform.Rotation=&RigVMFunction_MathQuaternionFromEuler(Euler=(X=0,Y=25,Z=0)).Result
    ItemToModify.1.Transform.Rotation=&RigVMFunction_MathQuaternionFromEuler(Euler=(X=0,Y=15,Z=-5)).Result
    ItemToModify.0.Transform.Rotation=&RigVMFunction_MathQuaternionFromEuler(Euler=(X=0,Y=0,Z=3)).Result
   )
   n9 = &RigVMFunction_AlphaInterp(InterpSpeedDecreasing=7, InterpSpeedIncreasing=5, Scale=1, bClampResult=False, Value=RigVMVariableNode() {variable=Skydive}.Value)
   > &RigUnit_ModifyTransforms(
    ItemToModify="((Item=(Type=Bone,Name=\"dyn_hair_side_fr_l_01\"),Transform=(Rotation=(X=0,Y=0,Z=0,W=1),Translation=(X=0,Y=0,Z=0),Scale3D=(X=1,Y=1,Z=1))),(Item=(Type=Bone,Name=\"dyn_hair_side_l_mstr\"),Transform=(Rotation=(X=0,Y=0,Z=0,W=1),Translation=(X=0,Y=0,Z=0),Scale3D=(X ...(truncated)"
    Weight=@n9.Result
    ItemToModify.0.Transform.Rotation=&RigVMFunction_MathQuaternionFromEuler(Euler=(X=0,Y=10,Z=20)).Result
    ItemToModify.1.Transform.Rotation=&RigVMFunction_MathQuaternionFromEuler(Euler=(X=0,Y=10,Z=20)).Result
    ItemToModify.2.Transform.Rotation=&RigVMFunction_MathQuaternionFromEuler(Euler=(X=0,Y=5,Z=0)).Result
    ItemToModify.3.Transform.Rotation=&RigVMFunction_MathQuaternionFromEuler(Euler=(X=0,Y=10,Z=0)).Result
   )
   &RigVMFunction_AlphaInterp(InterpSpeedDecreasing=7, InterpSpeedIncreasing=1, Scale=1, bClampResult=False, Value=@n2.OutputParam)
  ]}
]
hierarchy [
 root {loc=<0 0 0>, rot=<0 0 0 1>}
 pelvis {parent=root, loc=<0 -0.5301 98.0571>, rot=<0.0127 0.707 -0.0127 -0.707>}
 spine_01 {parent=pelvis, loc=<6.1208 -0.9328 0>, rot=<0 0 0.034 -0.9994>}
 spine_02 {parent=spine_01, loc=<6.5996 -0.2095 0>, rot=<0 0 0.0238 -0.9997>}
 spine_03 {parent=spine_02, loc=<6.4914 -0.1237 0>, rot=<0 0 0.0149 -0.9999>}
 spine_04 {parent=spine_03, loc=<6.9598 0.0794 0>, rot=<0 0 -0.0132 -0.9999>}
 spine_05 {parent=spine_04, loc=<6.9987 0 0>, rot=<0 0 -0.0015 -1>}
 neck_01 {parent=spine_05, loc=<11.3138 0.1529 0>, rot=<0 0 -0.2003 -0.9797>}
 neck_02 {parent=neck_01, loc=<4.5627 0 0>, rot=<0 0 0.0504 -0.9987>}
 head {parent=neck_02, loc=<5.3914 0 0>, rot=<0 0 0.115 -0.9934>}
 dyn_strand_fr_l_01 {parent=head, loc=<9.6854 8.7957 0.2026>, rot=<0.0541 -0.7206 -0.6185 -0.3086>}
 dyn_strand_fr_l_02 {parent=dyn_strand_fr_l_01, loc=<3.2631 0 0>, rot=<-0.0036 -0.1994 0.0183 -0.9797>}
 dyn_strand_fr_l_03 {parent=dyn_strand_fr_l_02, loc=<3.3519 0 0>, rot=<0.0296 -0.1922 -0.0689 -0.9785>}
 dyn_strand_l_01 {parent=head, loc=<6.0348 8.3054 -4.7929>, rot=<0.2405 -0.9035 -0.3317 -0.1259>}
 dyn_strand_l_02 {parent=dyn_strand_l_01, loc=<3.4947 -0.0031 0.0792>, rot=<0.0666 -0.1497 -0.0324 -0.986>}
 dyn_strand_l_03 {parent=dyn_strand_l_02, loc=<2.7957 0.3418 0.2487>, rot=<0.0917 -0.0917 -0.1651 -0.9777>}
 dyn_strand_fr_r_01 {parent=head, loc=<9.124 7.3959 3.9942>, rot=<0.1651 0.2577 0.9478 -0.0895>}
 dyn_strand_fr_r_02 {parent=dyn_strand_fr_r_01, loc=<3.3521 0.1505 -0.0002>, rot=<0.0061 -0.0427 -0.1266 -0.991>}
 dyn_strand_fr_r_03 {parent=dyn_strand_fr_r_02, loc=<3.5308 -0.0011 0.0079>, rot=<-0.0491 -0.0767 0.0246 -0.9955>}
 dyn_hair_side_fr_l_01 {parent=head, loc=<2.5839 2.9536 -7.0328>, rot=<0.0458 -0.9972 -0.0022 0.0583>}
 dyn_hair_side_fr_l_02 {parent=dyn_hair_side_fr_l_01, loc=<4.1837 0 0>, rot=<0 0 0 1>}
 dyn_hair_side_fr_l_03 {parent=dyn_hair_side_fr_l_02, loc=<4.1492 0 0>, rot=<0 0 0 1>}
 dyn_hair_side_fr_l_04 {parent=dyn_hair_side_fr_l_03, loc=<2.1556 0 0>, rot=<0 0 0 1>}
 dyn_hair_side_l_mstr {parent=head, loc=<1.9678 0.0384 -7.7297>, rot=<0.0576 -0.9975 -0.002 0.0401>}
 dyn_hair_side_l_01 {parent=dyn_hair_side_l_mstr, loc=<2.5944 0 0>, rot=<0.0015 -0.0143 0 -0.9999>}
 dyn_hair_side_l_fr_01 {parent=dyn_hair_side_l_01, loc=<1.4617 0.637 -0.1976>, rot=<0.0009 -0.0145 -0.0462 -0.9988>}
 dyn_hair_side_l_fr_02 {parent=dyn_hair_side_l_fr_01, loc=<2.6073 0 0>, rot=<0 0 0 1>}
 dyn_hair_side_l_fr_03 {parent=dyn_hair_side_l_fr_02, loc=<3.4815 0 0>, rot=<0 0 0 1>}
 dyn_hair_side_l_fr_04 {parent=dyn_hair_side_l_fr_03, loc=<1.8873 0 0>, rot=<0 0 0 1>}
 dyn_hair_side_l_bk_02 {parent=dyn_hair_side_l_01, loc=<1.2701 -0.4799 0.0314>, rot=<-0.0034 0.0388 -0.0203 -0.999>}
 dyn_hair_side_l_bk_01 {parent=dyn_hair_side_l_bk_02, loc=<2.9581 0 0>, rot=<0 0 0 1>}
 dyn_hair_side_r_mstr {parent=head, loc=<2.1066 1.1052 7.1416>, rot=<-0.016 0.005 0.9995 -0.0274>}
 dyn_hair_side_r_01 {parent=dyn_hair_side_r_mstr, loc=<2.4634 -0.0435 -0.0126>, rot=<0.0004 -0.0058 0.02 -0.9998>}
 dyn_hair_side_r_fr_01 {parent=dyn_hair_side_r_01, loc=<1.0456 -0.7913 -0.4473>, rot=<0.004 -0.0026 -0.0233 -0.9997>}
 dyn_hair_side_r_fr_02 {parent=dyn_hair_side_r_fr_01, loc=<2.9901 0 0>, rot=<0 0 0 1>}
 dyn_hair_side_r_fr_03 {parent=dyn_hair_side_r_fr_02, loc=<2.8031 0 0>, rot=<0 0 0 1>}
 dyn_hair_side_r_fr_04 {parent=dyn_hair_side_r_fr_03, loc=<2.7389 0 0>, rot=<0 0 0 1>}
 dyn_hair_side_r_bk_01 {parent=dyn_hair_side_r_01, loc=<0.0501 0.8507 0.2217>, rot=<0.0055 0.0083 -0.1583 -0.9873>}
 dyn_hair_side_r_bk_02 {parent=dyn_hair_side_r_bk_01, loc=<1.3164 0 0>, rot=<0 0 0 1>}
 dyn_hair_side_r_bk_03 {parent=dyn_hair_side_r_bk_02, loc=<1.8015 0 0>, rot=<0 0 0 1>}
 dyn_hair_side_r_bk_04 {parent=dyn_hair_side_r_bk_03, loc=<1.8788 0 0>, rot=<0 0 0 1>}
 dyn_hair_side_r_mid_01 {parent=head, loc=<0.8909 0.1693 5.5265>, rot=<-0.0058 0.0002 0.9987 -0.0506>}
 dyn_hair_side_r_mid_02 {parent=dyn_hair_side_r_mid_01, loc=<2.0211 0 0>, rot=<0 0 0 1>}
 dyn_hair_side_r_mid_03 {parent=dyn_hair_side_r_mid_02, loc=<2.7262 0 0>, rot=<0 0 0 1>}
 dyn_hair_side_r_mid_04 {parent=dyn_hair_side_r_mid_03, loc=<2.5344 0 0>, rot=<0 0 0 1>}
 dyn_hair_bun {parent=head, loc=<1.5792 -7.291 -0.1527>, rot=<0.4477 -0.8933 0.0177 -0.036>}
 dyn_bun_top_l {parent=dyn_hair_bun, loc=<2.4347 -3.9142 2.9779>, rot=<0.5882 -0.8037 0.0255 0.0858>}
 dyn_bun_top_r_01 {parent=dyn_hair_bun, loc=<0.5306 -5.0931 -4.3436>, rot=<-0.8621 0.0364 -0.0334 -0.5044>}
 dyn_bun_top_r_02 {parent=dyn_bun_top_r_01, loc=<1.4706 -0.009 0.0552>, rot=<-0.1024 -0.3269 -0.0174 -0.9393>}
 dyn_bun_l_01 {parent=dyn_hair_bun, loc=<1.162 3.5723 4.9542>, rot=<0.4531 0.0167 -0.265 -0.851>}
 dyn_bun_l_02 {parent=dyn_bun_l_01, loc=<2.4939 -0.0154 0.2002>, rot=<-0.0278 -0.332 -0.0157 -0.9427>}
 dyn_hair_bk_l_01 {parent=head, loc=<-2.375 -4.2301 -3.1929>, rot=<0.0648 -0.9109 0.4074 0.0062>}
 dyn_hair_bk_l_02 {parent=dyn_hair_bk_l_01, loc=<1.8555 0 0>, rot=<0 0 0 1>}
 dyn_hair_bk_l_03 {parent=dyn_hair_bk_l_02, loc=<1.6579 0 0>, rot=<0 0 0 1>}
 dyn_hair_bk_l_mid_01 {parent=head, loc=<-3.5697 -5.0482 -2.2278>, rot=<0.2404 -0.824 0.513 -0.0012>}
 dyn_hair_bk_l_mid_02 {parent=dyn_hair_bk_l_mid_01, loc=<1.8555 0 0>, rot=<0 0 0 1>}
 dyn_hair_bk_l_mid_03 {parent=dyn_hair_bk_l_mid_02, loc=<1.6579 0 0>, rot=<0 0 0 1>}
 dyn_hair_bk_l_mid_04 {parent=dyn_hair_bk_l_mid_03, loc=<1.2385 0 0>, rot=<0 0 0 1>}
 dyn_hair_bk_r_01 {parent=head, loc=<-2.2308 -5.4003 2.5567>, rot=<0.0079 -0.4566 0.8667 0.2009>}
 dyn_hair_bk_r_02 {parent=dyn_hair_bk_r_01, loc=<2.5415 0 0>, rot=<0.0084 0.1467 -0.3371 -0.9299>}
 dyn_hair_bk_r_03 {parent=dyn_hair_bk_r_02, loc=<1.7814 0 0>, rot=<-0.172 0.0512 -0.3743 -0.9098>}
 dyn_hair_bk_r_04 {parent=dyn_hair_bk_r_03, loc=<1.5423 0 0>, rot=<0.0612 -0.0565 0.1247 -0.9887>}
 dyn_hair_bk_r_mid_01 {parent=head, loc=<-4.2274 -5.4482 1.3512>, rot=<0.0193 -0.5733 0.805 -0.1515>}
 dyn_hair_bk_r_mid_02 {parent=dyn_hair_bk_r_mid_01, loc=<1.4496 0 0>, rot=<0 0 0 1>}
 dyn_hair_bk_r_mid_03 {parent=dyn_hair_bk_r_mid_02, loc=<1.8632 0 0>, rot=<0 0 0 1>}
 dyn_hair_bk_r_mid_04 {parent=dyn_hair_bk_r_mid_03, loc=<2.0319 0 0>, rot=<0 0 0 1>}
]
variables [
 ADS {type=double, default=0}
 Crouching {type=double, default=0}
 Skydive {type=double, default=0}
]
