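# Physics asset for the F_MED_HighBeam face accessory mesh, setting up physics for its side hair and loose strands.
# Small capsule chains branch from the head or left/right master bones, mostly with tight swing and near-zero twist limits.
# A head sphere and neck and upper-spine capsules form the body, linked by free-rotation constraints.
# The left master also rotates freely, and the left back chain runs in reverse order, 02 before 01.
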PhysicsAsset /BRCosmetics/Characters/Player/Female/Medium/Bodies/F_MED_HighBeam/Meshes/Parts/F_MED_HighBeam_FaceAcc_Physics
bodies {
 dyn_hair_side_fr_l_01 [{shapeType=Capsule, length=3, radius=1, shapeName=dyn_hair_side_fr_l_01_capsule, center=<1 0 0>, rotation=<rot 90 0 0>}]
 dyn_hair_side_fr_l_02 [{shapeType=Capsule, length=3, radius=1, shapeName=dyn_hair_side_fr_l_02_capsule, center=<1 0 0>, rotation=<rot 90 0 0>}]
 dyn_hair_side_fr_l_03 [{shapeType=Capsule, length=3, radius=1, shapeName=dyn_hair_side_fr_l_03_capsule, center=<1 0 0>, rotation=<rot 90 0 0>}]
 dyn_hair_side_fr_l_04 [{shapeType=Capsule, length=3, radius=1, shapeName=dyn_hair_side_fr_l_04_capsule, center=<1 0 0>, rotation=<rot 90 0 0>}]
 dyn_hair_side_l_01 [{shapeType=Capsule, length=2, radius=1, shapeName=dyn_hair_side_l_01_capsule, center=<1 0 0>, rotation=<rot 90 0 0>}]
 dyn_hair_side_l_bk_01 [{shapeType=Capsule, length=2, radius=1, shapeName=dyn_hair_side_l_bk_01_capsule, center=<1 0 0>, rotation=<rot 90 0 0>}]
 dyn_hair_side_l_bk_02 [{shapeType=Capsule, length=2, radius=1, shapeName=dyn_hair_side_l_bk_02_capsule, center=<1 0 0>, rotation=<rot 90 0 0>}]
 dyn_hair_side_l_fr_01 [{shapeType=Capsule, length=2, radius=1, shapeName=dyn_hair_side_l_fr_01_capsule, center=<1 0 0>, rotation=<rot 90 0 0>}]
 dyn_hair_side_l_fr_02 [{shapeType=Capsule, length=2, radius=1, shapeName=dyn_hair_side_l_fr_02_capsule, center=<1 0 0>, rotation=<rot 90 0 0>}]
 dyn_hair_side_l_fr_03 [{shapeType=Capsule, length=2, radius=1, shapeName=dyn_hair_side_l_fr_03_capsule, center=<1 0 0>, rotation=<rot 90 0 0>}]
 dyn_hair_side_l_fr_04 [{shapeType=Capsule, length=2, radius=1, shapeName=dyn_hair_side_l_fr_04_capsule, center=<1 0 0>, rotation=<rot 90 0 0>}]
 dyn_hair_side_l_mstr [{shapeType=Capsule, length=2, radius=1, shapeName=dyn_hair_side_l_mstr_capsule, rotation=<rot 90 0 0>}]
 dyn_hair_side_r_01 [{shapeType=Capsule, length=2, radius=1, shapeName=dyn_hair_side_r_01_capsule, center=<1 0 0>, rotation=<rot 90 0 0>}]
 dyn_hair_side_r_bk_01 [{shapeType=Capsule, length=2, radius=1, shapeName=dyn_hair_side_r_bk_01_capsule, center=<1 0 0>, rotation=<rot 90 0 0>}]
 dyn_hair_side_r_bk_02 [{shapeType=Capsule, length=2, radius=1, shapeName=dyn_hair_side_r_bk_02_capsule, center=<1 0 0>, rotation=<rot 90 0 0>}]
 dyn_hair_side_r_bk_03 [{shapeType=Capsule, length=2, radius=1, shapeName=dyn_hair_side_r_bk_03_capsule, center=<1 0 0>, rotation=<rot 90 0 0>}]
 dyn_hair_side_r_bk_04 [{shapeType=Capsule, length=2, radius=1, shapeName=dyn_hair_side_r_bk_04_capsule, center=<1 0 0>, rotation=<rot 90 0 0>}]
 dyn_hair_side_r_fr_01 [{shapeType=Capsule, length=2, radius=1, shapeName=dyn_hair_side_r_fr_01_capsule, center=<1 0 0>, rotation=<rot 90 0 0>}]
 dyn_hair_side_r_fr_02 [{shapeType=Capsule, length=2, radius=1, shapeName=dyn_hair_side_r_fr_02_capsule, center=<1 0 0>, rotation=<rot 90 0 0>}]
 dyn_hair_side_r_fr_03 [{shapeType=Capsule, length=2, radius=1, shapeName=dyn_hair_side_r_fr_03_capsule, center=<1 0 0>, rotation=<rot 90 0 0>}]
 dyn_hair_side_r_fr_04 [{shapeType=Capsule, length=2, radius=1, shapeName=dyn_hair_side_r_fr_04_capsule, center=<1 0 0>, rotation=<rot 90 0 0>}]
 dyn_hair_side_r_mid_01 [{shapeType=Capsule, length=2, radius=1, shapeName=dyn_hair_side_r_mid_01_capsule, center=<1 0 0>, rotation=<rot 90 0 0>}]
 dyn_hair_side_r_mid_02 [{shapeType=Capsule, length=2, radius=1, shapeName=dyn_hair_side_r_mid_02_capsule, center=<1 0 0>, rotation=<rot 90 0 0>}]
 dyn_hair_side_r_mid_03 [{shapeType=Capsule, length=2, radius=1, shapeName=dyn_hair_side_r_mid_03_capsule, center=<1 0 0>, rotation=<rot 90 0 0>}]
 dyn_hair_side_r_mid_04 [{shapeType=Capsule, length=2, radius=1, shapeName=dyn_hair_side_r_mid_04_capsule, center=<1 0 0>, rotation=<rot 90 0 0>}]
 dyn_hair_side_r_mstr [{shapeType=Capsule, length=2, radius=1, shapeName=dyn_hair_side_r_mstr_capsule, center=<1 0 0>, rotation=<rot 90 0 0>}]
 dyn_strand_fr_l_01 [{shapeType=Capsule, length=2, radius=1, shapeName=dyn_strand_fr_l_01_capsule, rotation=<rot 90 0 0>}]
 dyn_strand_fr_l_02 [{shapeType=Capsule, length=2, radius=1, shapeName=dyn_strand_fr_l_02_capsule, center=<1 0 0>, rotation=<rot 90 0 0>}]
 dyn_strand_fr_l_03 [{shapeType=Capsule, length=2, radius=1, shapeName=dyn_strand_fr_l_03_capsule, center=<1 0 0>, rotation=<rot 90 0 0>}]
 dyn_strand_fr_r_01 [{shapeType=Capsule, length=2, radius=1, shapeName=dyn_strand_fr_r_01_capsule, center=<1 0 0>, rotation=<rot 90 0 0>}]
 dyn_strand_fr_r_02 [{shapeType=Capsule, length=2, radius=1, shapeName=dyn_strand_fr_r_02_capsule, center=<1 0 0>, rotation=<rot 90 0 0>}]
 dyn_strand_fr_r_03 [{shapeType=Capsule, length=2, radius=1, shapeName=dyn_strand_fr_r_03_capsule, center=<1 0 0>, rotation=<rot 90 0 0>}]
 dyn_strand_l_01 [{shapeType=Capsule, length=2, radius=1, shapeName=dyn_strand_l_01_capsule, center=<1 0 0>, rotation=<rot 90 0 0>}]
 dyn_strand_l_02 [{shapeType=Capsule, length=2, radius=1, shapeName=dyn_strand_l_02_capsule, center=<1 0 0>, rotation=<rot 90 0 0>}]
 dyn_strand_l_03 [{shapeType=Capsule, length=2, radius=1, shapeName=dyn_strand_l_03_capsule, center=<1 0 0>, rotation=<rot 90 0 0>}]
 head [{shapeType=Sphere, radius=7.3563, shapeName=head_sphere, center=<3.9824 0 0>}]
 neck_01 [{shapeType=Capsule, length=2.3657, radius=5.6343, shapeName=neck_01_capsule, rotation=<rot -90 -179.8418 179.8416>}]
 neck_02 [{shapeType=Capsule, length=2.3657, radius=5.6343, shapeName=neck_02_capsule, rotation=<rot -90 0 -0.0002>}]
 spine_05 [{shapeType=Capsule, length=24.8811, radius=8.7516, shapeName=spine_05_capsule, center=<0.9586 -1.3605 0>, rotation=<rot 0 -90 0>}]
}
constraints [
 {bone1Name=dyn_hair_side_r_fr_04, bone2Name=dyn_hair_side_r_fr_03, swing1LimitDegrees=25, swing2LimitDegrees=15, twistLimitDegrees=1}
 {bone1Name=dyn_hair_side_r_01, bone2Name=dyn_hair_side_r_mstr, swing1LimitDegrees=25, swing2LimitDegrees=15, twistLimitDegrees=1}
 {bone1Name=dyn_hair_side_r_bk_01, bone2Name=dyn_hair_side_r_01, swing1LimitDegrees=25, swing2LimitDegrees=15, twistLimitDegrees=1}
 {bone1Name=dyn_hair_side_r_bk_02, bone2Name=dyn_hair_side_r_bk_01, swing1LimitDegrees=25, swing2LimitDegrees=15, twistLimitDegrees=1}
 {bone1Name=dyn_hair_side_r_bk_03, bone2Name=dyn_hair_side_r_bk_02, swing1LimitDegrees=25, swing2LimitDegrees=15, twistLimitDegrees=1}
 {bone1Name=dyn_hair_side_r_bk_04, bone2Name=dyn_hair_side_r_bk_03, swing1LimitDegrees=25, swing2LimitDegrees=15, twistLimitDegrees=1}
 {bone1Name=dyn_hair_side_r_fr_01, bone2Name=dyn_hair_side_r_01, swing1LimitDegrees=25, swing2LimitDegrees=15, twistLimitDegrees=1}
 {bone1Name=dyn_hair_side_r_fr_02, bone2Name=dyn_hair_side_r_fr_01, swing1LimitDegrees=25, swing2LimitDegrees=15, twistLimitDegrees=1}
 {bone1Name=dyn_hair_side_r_fr_03, bone2Name=dyn_hair_side_r_fr_02, swing1LimitDegrees=25, swing2LimitDegrees=15, twistLimitDegrees=1}
 {bone1Name=dyn_hair_side_r_mstr, bone2Name=head, swing1LimitDegrees=15, swing2LimitDegrees=5, twistLimitDegrees=1}
 {bone1Name=dyn_strand_fr_l_01, bone2Name=head, swing1LimitDegrees=15, swing2LimitDegrees=10, twistLimitDegrees=1}
 {bone1Name=dyn_strand_fr_l_02, bone2Name=dyn_strand_fr_l_01, swing1LimitDegrees=15, swing2LimitDegrees=10, twistLimitDegrees=1}
 {bone1Name=dyn_strand_fr_l_03, bone2Name=dyn_strand_fr_l_02, swing1LimitDegrees=15, swing2LimitDegrees=10, twistLimitDegrees=1}
 {bone1Name=dyn_strand_fr_r_01, bone2Name=head, swing1LimitDegrees=15, swing2LimitDegrees=10, twistLimitDegrees=1}
 {bone1Name=dyn_strand_fr_r_02, bone2Name=dyn_strand_fr_r_01, swing1LimitDegrees=15, swing2LimitDegrees=10, twistLimitDegrees=1}
 {bone1Name=dyn_strand_fr_r_03, bone2Name=dyn_strand_fr_r_02, swing1LimitDegrees=15, swing2LimitDegrees=10, twistLimitDegrees=1}
 {bone1Name=dyn_strand_l_01, bone2Name=head, swing1LimitDegrees=15, swing2LimitDegrees=10, twistLimitDegrees=1}
 {bone1Name=dyn_strand_l_02, bone2Name=dyn_strand_l_01, swing1LimitDegrees=15, swing2LimitDegrees=10, twistLimitDegrees=1}
 {bone1Name=dyn_strand_l_03, bone2Name=dyn_strand_l_02, swing1LimitDegrees=15, swing2LimitDegrees=10, twistLimitDegrees=1}
 {
  bone1Name=dyn_hair_side_l_mstr
  bone2Name=head
  swing1LimitDegrees=45
  swing1Motion=Free
  swing2LimitDegrees=45
  swing2Motion=Free
  twistLimitDegrees=45
  twistMotion=Free
 }
 {bone1Name=dyn_hair_side_l_01, bone2Name=dyn_hair_side_l_mstr, swing1LimitDegrees=20, swing2LimitDegrees=10, twistLimitDegrees=1}
 {bone1Name=dyn_hair_side_l_bk_02, bone2Name=dyn_hair_side_l_01, swing1LimitDegrees=25, swing2LimitDegrees=15, twistLimitDegrees=1}
 {bone1Name=dyn_hair_side_l_bk_01, bone2Name=dyn_hair_side_l_bk_02, swing1LimitDegrees=25, swing2LimitDegrees=15, twistLimitDegrees=1}
 {bone1Name=dyn_hair_side_l_fr_01, bone2Name=dyn_hair_side_l_01, swing1LimitDegrees=25, swing2LimitDegrees=15, twistLimitDegrees=1}
 {bone1Name=dyn_hair_side_l_fr_02, bone2Name=dyn_hair_side_l_fr_01, swing1LimitDegrees=25, swing2LimitDegrees=15, twistLimitDegrees=1}
 {bone1Name=dyn_hair_side_l_fr_03, bone2Name=dyn_hair_side_l_fr_02, swing1LimitDegrees=25, swing2LimitDegrees=15, twistLimitDegrees=1}
 {bone1Name=dyn_hair_side_l_fr_04, bone2Name=dyn_hair_side_l_fr_03, swing1LimitDegrees=25, swing2LimitDegrees=15, twistLimitDegrees=1}
 {bone1Name=dyn_hair_side_r_mid_01, bone2Name=head, swing1LimitDegrees=25, swing2LimitDegrees=15, twistLimitDegrees=1}
 {bone1Name=dyn_hair_side_r_mid_02, bone2Name=dyn_hair_side_r_mid_01, swing1LimitDegrees=25, swing2LimitDegrees=15, twistLimitDegrees=1}
 {bone1Name=dyn_hair_side_r_mid_03, bone2Name=dyn_hair_side_r_mid_02, swing1LimitDegrees=25, swing2LimitDegrees=15, twistLimitDegrees=1}
 {bone1Name=dyn_hair_side_r_mid_04, bone2Name=dyn_hair_side_r_mid_03, swing1LimitDegrees=25, swing2LimitDegrees=15, twistLimitDegrees=1}
 {
  bone1Name=head
  bone2Name=neck_02
  swing1LimitDegrees=45
  swing1Motion=Free
  swing2LimitDegrees=45
  swing2Motion=Free
  twistLimitDegrees=45
  twistMotion=Free
 }
 {
  bone1Name=neck_02
  bone2Name=neck_01
  swing1LimitDegrees=45
  swing1Motion=Free
  swing2LimitDegrees=45
  swing2Motion=Free
  twistLimitDegrees=45
  twistMotion=Free
 }
 {bone1Name=dyn_hair_side_fr_l_01, bone2Name=head, swing1LimitDegrees=5, swing2LimitDegrees=5, twistLimitDegrees=1}
 {bone1Name=dyn_hair_side_fr_l_02, bone2Name=dyn_hair_side_fr_l_01, swing1LimitDegrees=25, swing2LimitDegrees=15, twistLimitDegrees=1}
 {bone1Name=dyn_hair_side_fr_l_03, bone2Name=dyn_hair_side_fr_l_02, swing1LimitDegrees=25, swing2LimitDegrees=15, twistLimitDegrees=1}
 {bone1Name=dyn_hair_side_fr_l_04, bone2Name=dyn_hair_side_fr_l_03, swing1LimitDegrees=25, swing2LimitDegrees=15, twistLimitDegrees=1}
 {
  bone1Name=neck_01
  bone2Name=spine_05
  swing1LimitDegrees=45
  swing1Motion=Free
  swing2LimitDegrees=45
  swing2Motion=Free
  twistLimitDegrees=45
  twistMotion=Free
 }
]
